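; Parent Creative spawner device for Dagwood (Valet car) vehicles, applying user options to the vehicle it spawns.
; On spawn it extends parent logic to remove the radio, set self-destruct via a gameplay effect, pick color, and apply authority-only tire and cow catcher mods by tag.
; UpdateVehicleConfigs pushes boost, damage and destruction settings, disables fade-out, and sets Valet explosion damage.
; Scalable-float overrides gate radio, self-destruct and mod options.
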(event EventOnVehicleSpawned
  (params (Spawned%20Vehicle :type Fort%20Athena%20Vehicle%20Object%20Reference))
  ;[0.09063 0.04957 1] Vehicle Spawning Logic -- Accounting for Dagwood-specific options
  (|CallParentFunction Spawned%20Vehicle)
  (CallFunction|HandleVehicleRadio (and (Variables|UserKnobs|GetbRadioEnabled) (ScalableFloat|GetValueasBool (Variables|Default|GetRadioOverride)))) ; disable radio
  (CallFunction|HandleModSettings)
  (CallFunction|HandleSelfDestruct (and (Variables|UserKnobs|GetbAllowSelfDestruct) (ScalableFloat|GetValueasBool (Variables|Default|GetAllowSelfDestructOverride))))
  (if (>= (Variables|UserKnobs|GetColorOption) 0)
    (CallFunction|CheckColorSettings)))

(fn CheckColorSettings
  ;[0.1451 1 0.6491 0.75] Override random material determination and set a specific one based on the vehicle
  (bind _asvalet_basic_car_vehicle (Utilities|Casting|CastToValet_BasicCar_Vehicle (Variables|Default|GetCachedVehicleInterface)))
  (Variables|/Valet/BasicCar/Valet_BasicCar_Vehicle.Valet_BasicCar_Vehicle_C|SetRandomMatInt (Variables|UserKnobs|GetColorOption) _asvalet_basic_car_vehicle))

(fn HandleCowCatcherSettings
  (bind _cowcatchermodtag (Variables|Default|GetCowCatcherModTag))
  ;[1 0.7311 0.6711 0.75] If Cow Catcher is enabled, then apply the Cow Catcher to the vehicle.
  (Networking|SwitchHasAuthority
    (:Authority
      (if (and (ScalableFloat|GetValueasBool (Variables|Default|GetCowCatcherOverride)) (Variables|Default|GetCowCatcherOption))
        (bind _asfort_dagwood_vehicle (Utilities|Casting|CastToFortDagwoodVehicle (Variables|Default|GetCachedVehicleInterface)))
        (if (GameplayTags|IsGameplayTagValid _cowcatchermodtag)
          (Vehicle|Mod|ApplyVehicleMod _asfort_dagwood_vehicle _cowcatchermodtag 0))))))

(fn HandleModSettings
  (CallFunction|HandleTireSettings)
  (CallFunction|HandleCowCatcherSettings))

(fn HandleSelfDestruct
  (params (Allow :type Boolean))
  (bind _ge_allowselfdestruct (Variables|Default|GetGE_AllowSelfDestruct))
  (bind _asfort_dagwood_vehicle (Utilities|Casting|CastToFortDagwoodVehicle (Variables|Default|GetCachedVehicleInterface)))
  (if Allow
    (GameplayEffects|ApplyGameplayEffectToSelf :self (Ability|GetAbilitySystemComponent _asfort_dagwood_vehicle) :GameplayEffectClass _ge_allowselfdestruct :Level 0.0)
    (else
      (GameplayEffects|RemoveActiveGameplayEffectbySourceEffect (Ability|GetAbilitySystemComponent _asfort_dagwood_vehicle) _ge_allowselfdestruct 0 -1))))

(fn HandleTireSettings
  (bind _tiremodtag (Variables|Default|GetTireModTag))
  ;[1 0.7311 0.6711 0.75] If our tires are not Default (0), then set the tires to the proper mod
  (Networking|SwitchHasAuthority
    (:Authority
      (switch int (Math|Integer|SelectInt (Variables|Default|GetTireOption) 0 (ScalableFloat|GetValueasBool (Variables|Default|GetTireOverride)))
        (:1
          (bind _asfort_dagwood_vehicle (Utilities|Casting|CastToFortDagwoodVehicle (Variables|Default|GetCachedVehicleInterface)))
          (if (GameplayTags|IsGameplayTagValid _tiremodtag)
            (Vehicle|Mod|ApplyVehicleMod _asfort_dagwood_vehicle _tiremodtag 0) ; Used instead of ApplyTireMod for proper internal setup
          ))))))

(fn HandleVehicleRadio
  (params (bRadioIsEnabled :type Boolean))
  ;[0.9583 0.14 0.04445] Handle Radio by setting.  At this time, we only destroy them, but we are keeping this in place in the event we want to add a radio component to non-Valet-derived vehicles in the future.
  (if (not bRadioIsEnabled)
    (bind _asvalet_basic_car_vehicle (Utilities|Casting|CastToValet_BasicCar_Vehicle (Variables|Default|GetCachedVehicleInterface)))
    (Utilities|IsValid (Actor|GetComponentbyClass _asvalet_basic_car_vehicle "/Script/SrirachaRanch.StreamingRadioPlayerComponent")
      (:Is%20Valid
        (Components|DestroyComponent (Actor|GetComponentbyClass _asvalet_basic_car_vehicle "/Script/SrirachaRanch.StreamingRadioPlayerComponent"))))))

(fn UpdateVehicleConfigs
  ;[0.4608 1 0.3476] Set the vehicle runtime settings based on the devices user options
  (|CallParentFunction)
  (bind _asfort_dagwood_vehicle (Utilities|Casting|CastToFortDagwoodVehicle (Variables|Default|GetCachedVehicleInterface)))
  (Class|FortDagwoodVehicle|SetRuntimeBoostInfoonServer _asfort_dagwood_vehicle (Utilities|Struct|MakeRuntimeBoostInfo (select (Variables|Default|GetbHasInfiniteFuel) (!= (Variables|UserKnobs|GetBoostRegenSetting) 0) (>= (Variables|Default|GetBoostFuelUseMultiplier) 0)) (Variables|Default|GetbHasInfiniteFuel) (Variables|Default|GetBoost%20Push%20Force) (Variables|Default|GetBoost%20Top%20Speed%20Force%20Multiplier) (Variables|Default|GetBoost%20Top%20Speed%20Multiplier) (Variables|UserKnobs|GetBoostRegenSetting) (Math|Float|SelectFloat 0.0 (Variables|Default|GetBoost%20Rate%20of%20Use) (>= (Variables|UserKnobs|GetBoostRegenSetting) 1.0))))
  (Class|FortDagwoodVehicle|SetDamageInfoonServer _asfort_dagwood_vehicle (Variables|UserKnobs|GetbDamageFriendlyFire) (Variables|UserKnobs|GetbDamageOtherVehicles) (Variables|UserKnobs|GetbDamageOwnVehicle) (Variables|UserKnobs|GetDamage%20Allowed%20from%20Other%20Vehicle))
  ;;[1 0.9655 0.1244 1 12] Disable fade out by setting the duration to -1. Fade out is not necessary for Creative and makes some of the user settings confusing (MaxTimeExplosion and MaxTimeZeroHealth).
  (NativeFade|ServerSetFadeDuration _asfort_dagwood_vehicle -1.0)
  (Class|FortAthenaVehicle|ServerSetDestructionInfo _asfort_dagwood_vehicle (Variables|UserKnobs|GetSelfDestructTime) (Variables|UserKnobs|GetLifetimeAfterExplosion) (Variables|UserKnobs|GetSelfDestructCleanupTime) (Variables|UserKnobs|GetExplosionDelay))
  (bind _asvalet_basic_car_vehicle (Utilities|Casting|CastToValet_BasicCar_Vehicle _asfort_dagwood_vehicle))
  (Utilities|IsValid _asvalet_basic_car_vehicle
    (:Is%20Valid
      (Variables|/Valet/BasicCar/Valet_BasicCar_Vehicle.Valet_BasicCar_Vehicle_C|SetCachedExplosionDamageBuildings (Variables|UserKnobs|GetExplosionDamageToEnvironment) _asvalet_basic_car_vehicle)
      (Variables|/Valet/BasicCar/Valet_BasicCar_Vehicle.Valet_BasicCar_Vehicle_C|SetCachedExplosionDamageVehicles (Variables|UserKnobs|GetExplosionDamageToVehicles) _asvalet_basic_car_vehicle)
      (Variables|/Valet/BasicCar/Valet_BasicCar_Vehicle.Valet_BasicCar_Vehicle_C|SetCachedExplosionDamagePlayers (Variables|UserKnobs|GetExplosionDamageToPlayers) _asvalet_basic_car_vehicle))))

(fn ConstructionScript
  (|CallParentFunction))
