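; Parent Creative device that spawns, configures, respawns and destroys one vehicle on a spawn pad, following enabled state and minigame phase.
; Spawning runs on authority through the Creative throttle manager, with an overlap test and randomized anti-spam delay.
; Receiver events drive enable, respawn, destroy and driver seating, including deferred seating after game-start respawn; vehicle delegates broadcast gameplay events.
; Replicated cooldown start drives pad color and flicker; placement uses preview and overlap blocking.
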(event EventInitializeDevice
  (bind _returnvalue (Building|IsPreviewBuildingActor))
  ;[0.1083 1 0.0636] Initialization
  (|CallParentFunction)
  (CallFunction|HandleEditModeCollision)
  (CallFunction|HandleEditorVisibility)
  (if _returnvalue
    (CallFunction|HandleVehicleReferenceVisibility _returnvalue)
    (else
      (CallFunction|HandleVehicleReferenceVisibility _returnvalue)
      (Class|CreativeEnabledComponent|SetEnabledOnPhase (Variables|Default|GetEnabledComponent) "PreGame" (Variables|LegacyOptions|GetEnabledIndex))
      (CallFunction|UpdateVehicleSpawnSettings false)
      ;[0.8827 0.1953 1 0.25] We also want a preview in editor, but it can't spawn, so use a mesh.
      (if (and (Editor|InEditor) (Variables|Default|GetUseEditorPreview))
        (bind _loaded_object (Creative|Loading|AsyncLoadEditorSafe (Variables|Default|GetPreviewVehicleMesh))
          (:Loaded
            (bind _asstatic_mesh (Utilities|Casting|CastToStaticMesh _loaded_object))
            (Components|StaticMesh|SetStaticMesh (Variables|Default|GetEditorVehiclePreview) _asstatic_mesh)))))))

(event EventEndPlay
  (params (EndPlayReason :type EEndPlayReason%20Enum :default "Destroyed"))
  (bind _cachedvehicleinterface (Variables|Default|GetCachedVehicleInterface))
  ;[1 0.1508 0] Destruction
  (Variables|Default|SetCachedInstigator 0)
  (Networking|SwitchHasAuthority
    (:Authority
      (Utilities|IsValid (Utilities|ToObject(Interface) _cachedvehicleinterface)
        (:Is%20Valid
          (if (<= (Utilities|Array|Length (Vehicle|GetAllPassengers _cachedvehicleinterface)) 0)
            (CallFunction|ManuallyDestroyVehicle)
            (else
              (Vehicle|SetDestroyonLastExit (Variables|Default|GetCachedVehicleInterface) true))))))))

(event EventDestroyed
  (bind _cachedvehicleinterface (Variables|Default|GetCachedVehicleInterface))
  ;[1 0.1508 0] Destruction
  (Networking|SwitchHasAuthority
    (:Authority
      (Utilities|IsValid (Utilities|ToObject(Interface) _cachedvehicleinterface)
        (:Is%20Valid
          (if (<= (Utilities|Array|Length (Vehicle|GetAllPassengers _cachedvehicleinterface)) 0)
            (CallFunction|ManuallyDestroyVehicle)
            (else
              (Vehicle|SetDestroyonLastExit (Variables|Default|GetCachedVehicleInterface) true))))))))

(event Custom|Blend%20Emissive%20Color
  (params (BlendColor :type Linear%20Color%20Structure))
  ;[0.6287 0 1] Blend the color of the spawn pad lights when switching between states (Disabled, enabled, or on cooldown)
  (if (Math|Color|NotEqual(LinearColor) BlendColor (Variables|ColorSwitches|GetCurrent%20Color))
    (Variables|ColorSwitches|SetOld%20Color (Variables|ColorSwitches|GetCurrent%20Color))
    (Variables|ColorSwitches|SetCurrent%20Color BlendColor)
    (bind (_direction _alpha) (Timeline|Color%20Switch%20Timeline PlayFromStart)
      (:Update
        (Rendering|Material|SetVectorParameterValue (Variables|Cosmetics|GetMID%20SpawnPlate) "Emissive" (Math|Color|Lerp(LinearColor) (Variables|ColorSwitches|GetOld%20Color) (Variables|ColorSwitches|GetCurrent%20Color) _alpha) :!class /Script/Engine.MaterialInstanceDynamic)))))

(event MinigameStateChangedDelegate(FortMinigameProgress)
  (params (Minigame :type Fort%20Minigame%20Object%20Reference) (NewMinigameState :type EFortMinigameState%20Enum :default "PreGame"))
  ;[0.04638 0.9003 1 0.5] Handle Minigame State Change
  (Class|CreativeEnabledComponent|SetEnabledOnPhase (Variables|Default|GetEnabledComponent) NewMinigameState (Variables|LegacyOptions|GetEnabledIndex))
  (if (Class|CreativeEnabledComponent|IsEnabled (Variables|Default|GetEnabledComponent))
    (switch Utilities|FlowControl|Switch|SwitchonEFortMinigameState NewMinigameState
      (:PreGame
        (CallFunction|UpdateVehicleSpawnSettings false))
      (:Setup
        _)
      (:Warmup
        _)
      (:InProgress
        (CallFunction|CheckSeatingChannel))
      (:PostRoundEnd
        _)
      (:PostGameReset
        _))
    (CallFunction|RefreshVehicleSpawn)
    (else
      (CallFunction|ManuallyDestroyVehicle))))

(event Custom|Completed%20Throttle
  (params (Succeeded :type Boolean))
  ;[0.09063 0.04957 1] Vehicle Spawning Logic
  (Variables|SpamProtection|SetWaiting%20For%20Throttle false)
  (if Succeeded
    (bind _area_clear (CallFunction|TestVehicleOverlap))
    (if _area_clear
      (CallFunction|AttemptSpawnVehicle) ; If the vehicle has already been throttled, try to spawn it now
      (else
        _))
    (else
      _))
  (Variables|SpamProtection|SetWaiting%20to%20Spawn true)
  (CallFunction|AttemptSpawnAfterSpamProtectionDelay) ; If the vehicle failed to throttled, try to spawn it with delay to avoid spam spawning
)

(event Custom|Attempt%20Spawn%20Vehicle
  ;[0.09063 0.04957 1] Vehicle Spawning Logic
  (if (Class|CreativeEnabledComponent|IsEnabled (Variables|Default|GetEnabledComponent))
    (if (Variables|SpamProtection|GetRecently%20Spawned)
      (Variables|SpamProtection|SetWaiting%20to%20Spawn true)
      (else
        (CallFunction|SpawnVehicle)))))

(event Custom|Attempt%20Spawn%20After%20Spam%20Protection%20Delay
  (bind (_spam_protection_min_max_time_x _spam_protection_min_max_time_y) (Math|Vector2D|BreakVector2D (Variables|GetSpam%20Protection%20Min/Max%20Time_X)))
  ;[0.09063 0.04957 1] Vehicle Spawning Logic
  ;;[1 0.984 0.1121 1 14] Waits a defined time period after spawning a vehicle, and then checks to see if anything has attempted to spawn a new one.
  ;; This is a buffer on the local device, stopping you mass-respawning vehicles.
  (Utilities|FlowControl|Delay (Math|Random|RandomFloatinRange _spam_protection_min_max_time_x _spam_protection_min_max_time_y))
  (Variables|SpamProtection|SetRecently%20Spawned false)
  (if (Variables|SpamProtection|GetWaiting%20to%20Spawn)
    (CallFunction|RespawnVehicleFromSettings)))

(event Custom|SpawnVehicle
  ;[0.09063 0.04957 1] Vehicle Spawning Logic
  (Networking|SwitchHasAuthority
    (:Authority
      (if (and (Fortnite|AreAthenaVehiclesEnabled) (Utilities|IsValidClass (Variables|Default|GetVehicleType)))
        (bind _returnvalue (Game|SpawnActorfromClass (Variables|Default|GetVehicleType) (Math|Transform|MakeTransform (.location (Transformation|GetWorldTransform (Variables|Default|GetSpawnLocation))) (.rotation (Transformation|GetWorldTransform (Variables|Default|GetSpawnLocation)))) "AlwaysSpawn" "SelectDefaultAtRuntime" 0 false 0))
        (bind _asfort_vehicle_interface (Utilities|Casting|CastToFortVehicleInterface _returnvalue))
        (Variables|Default|SetCachedVehicleInterface _asfort_vehicle_interface)
        (CallFunction|InitializeVehicleHealth _returnvalue)
        (InteractInterface|SetAllowInteractionBetweenFortVolumes _returnvalue false :!class /Script/FortniteGame.FortAthenaVehicle)
        (CallFunction|OnVehicleSpawned _returnvalue)))))

(event Custom|On%20Vehicle%20Spawned
  (params (Spawned%20Vehicle :type Fort%20Athena%20Vehicle%20Object%20Reference))
  ;[0.09063 0.04957 1] Vehicle Spawning Logic
  (Utilities|IsValid Spawned%20Vehicle
    (:Is%20Valid
      (Utilities|FlowControl|Sequence
        (:then_0
          (Variables|SpamProtection|SetWaiting%20to%20Spawn false)
          (Utilities|Time|ClearandInvalidateTimerbyHandle (Variables|Default|GetVehicle%20Spawn%20Timer))
          (Variables|SpamProtection|SetRecently%20Spawned true)
          (Vehicle|WakeUp Spawned%20Vehicle)
          (CallFunction|UpdateVehicleConfigs)
          (CallFunction|AttemptSpawnAfterSpamProtectionDelay))
        (:then_1
          (Game|BindEventtoOnEndPlay Spawned%20Vehicle (EventDispatchers|CreateEvent :!event Custom|Vehicle%20Destroyed))
          (Vehicle|AssignOnPawnFinishEnteringVehicle Spawned%20Vehicle (EventDispatchers|CreateEvent :!event Custom|On%20Pawn%20Entered%20Vehicle))
          (Vehicle|AssignOnPawnExitedVehicleSeat Spawned%20Vehicle (EventDispatchers|CreateEvent :!event Custom|On%20Pawn%20Exited%20Vehicle))
          (Health|BindEventtoOnHandleOutOfHealthDelegate Spawned%20Vehicle (EventDispatchers|CreateEvent :!event Custom|VehicleOutOfHealth)))
        (:then_2
          (GameplayEvent|BroadcastEvent (Variables|UserOptions|Events|GetOn%20Vehicle%20Spawns) 0 (Utilities|Struct|MakeGameplayEventVehicleSpawned (Variables|Default|GetCachedInstigator)))
          (Variables|Default|SetCachedInstigator 0)
          (bind _asfort_athena_vehicle (Utilities|Casting|CastToFortAthenaVehicle (Variables|Default|GetCachedVehicleInterface)))
          (Class|FortAthenaVehicle|SetSupportCosmeticWrap _asfort_athena_vehicle (Variables|UserKnobs|GetbVehicleSupportsWraps)) ; Function has build in check that determines \nwhether the vehicle type should support wraps
          (CallFunction|ChecktoSeatonStart))))))

(event Custom|Manually%20Destroy%20Vehicle
  (bind _cachedvehicleinterface (Variables|Default|GetCachedVehicleInterface))
  (bind _returnvalue (Utilities|ToObject(Interface) _cachedvehicleinterface))
  ;[1 0.3476 0] Vehicle Destroying Logic
  (Networking|SwitchHasAuthority
    (:Authority
      (Utilities|IsValid _returnvalue
        (:Is%20Valid
          (bind _asfort_athena_vehicle (Utilities|Casting|CastToFortAthenaVehicle _returnvalue))
          (CallFunction|RemoveVehicleBinds _asfort_athena_vehicle)
          (CallFunction|DestroySpawnedVehicle))) ; This check is done in cast the cached vehicle is pending kill
    )))

(event Custom|Remove%20Vehicle%20Binds
  (params (Vehicle :type Fort%20Athena%20Vehicle%20Object%20Reference))
  ;[1 0.3476 0] Vehicle Destroying Logic
  (Utilities|IsValid Vehicle
    (:Is%20Valid
      (Game|UnbindEventfromOnEndPlay Vehicle (delegate Custom|Vehicle%20Destroyed))
      (Vehicle|UnbindEventfromOnPawnFinishEnteringVehicle Vehicle (EventDispatchers|CreateEvent :!event Custom|On%20Pawn%20Entered%20Vehicle))
      (Vehicle|UnbindEventfromOnPawnExitedVehicleSeat Vehicle (EventDispatchers|CreateEvent :!event Custom|On%20Pawn%20Exited%20Vehicle)))))

(event Custom|Vehicle%20Destroyed
  (params (Actor :type Actor%20Object%20Reference) (EndPlayReason :type EEndPlayReason%20Enum :default "Destroyed"))
  ;[1 0.3476 0] Vehicle Destroying Logic
  (Networking|SwitchHasAuthority
    (:Authority
      (GameplayEvent|BroadcastEvent (Variables|UserOptions|Events|GetOn%20Vehicle%20Is%20Destroyed) 0 (Utilities|Struct|MakeGameplayEventVehicleDestroyed (Variables|Default|GetCachedInstigator)))
      (Variables|Default|SetCachedInstigator 0)
      (Variables|Default|SetCachedVehicleInterface 0)
      (CallFunction|AttemptRespawnVehicle))))

(event Custom|VehicleOutOfHealth
  (params (DeadVehicle :type Fort%20Athena%20Vehicle%20Object%20Reference) (InTags :type Gameplay%20Tag%20Container%20Structure%20%28by%20ref%29) (EventInstigator :type Controller%20Object%20Reference) (DamageCauser :type Actor%20Object%20Reference))
  ;[1 0.3476 0] Vehicle Destroying Logic
  (Utilities|IsValid EventInstigator
    (:Is%20Valid
      (Utilities|IsValid (Variables|Default|GetCachedInstigator)
        (:Is%20Not%20Valid
          (Variables|Default|SetCachedInstigator EventInstigator))))))

(event Custom|Attempt%20Respawn%20Vehicle
  (bind _spawndelay (Variables|UserKnobs|GetSpawnDelay))
  ;[1 0.3476 0] Vehicle Destroying Logic
  (if (or (== (Variables|UserKnobs|GetSpawnDelay) 0.0) (Variables|Default|GetInstant%20Respawn%20Override))
    (CallFunction|RespawnVehicleFromSettings)
    (elif (!= -1.0 _spawndelay)
      (bind _returnvalue (Utilities|Time|SetTimerbyFunctionName 0 "Respawn Vehicle From Settings" _spawndelay))
      (Variables|Default|SetVehicle%20Spawn%20Timer _returnvalue)
      (Variables|Default|SetLastCooldownStart (Fortnite|Sync|Time|GetServerWorldTimeSeconds)))))

(event Custom|Throttle%20New%20Vehicle%20On%20Vehicle%20End%20Play
  (params (Actor :type Actor%20Object%20Reference) (End%20Play%20Reason :type EEndPlayReason%20Enum :default "Destroyed"))
  ;[1 0.3476 0] Vehicle Destroying Logic
  (CallFunction|ThrottleNewVehicle))

(event Custom|EnabledOnReceived
  (params (Instigator :type Controller%20Object%20Reference))
  (bind _enabledcomponent (Variables|Default|GetEnabledComponent))
  ;[0.3018 0 0.5312] Receivers
  (Variables|Default|SetCachedInstigator Instigator)
  (if (not (Class|CreativeEnabledComponent|IsEnabled (Variables|Default|GetEnabledComponent)))
    (Utilities|StopOtherforDelay :!pinned true
      (:Out%20A
        (Class|CreativeEnabledComponent|SetEnabled _enabledcomponent true))
      (:Out%20B
        (Class|CreativeEnabledComponent|SetEnabled _enabledcomponent false))) ; When one input is triggered, pass through\nand stop the other input passing through\nfor VehicleMinToggleTime seconds
    (CallFunction|UpdateVehicleSpawnSettings true)))

(event Custom|DisableOnReceived
  (params (Instigator :type Controller%20Object%20Reference))
  (bind _enabledcomponent (Variables|Default|GetEnabledComponent))
  ;[0.3018 0 0.5312] Receivers
  (Variables|Default|SetCachedInstigator Instigator)
  (if (Class|CreativeEnabledComponent|IsEnabled (Variables|Default|GetEnabledComponent))
    (Utilities|StopOtherforDelay :!pinned true
      (:Out%20A
        (Class|CreativeEnabledComponent|SetEnabled _enabledcomponent true))
      (:Out%20B
        (Class|CreativeEnabledComponent|SetEnabled _enabledcomponent false))) ; When one input is triggered, pass through\nand stop the other input passing through\nfor VehicleMinToggleTime seconds
    (CallFunction|UpdateVehicleSpawnSettings true)))

(event Custom|RespawnOnReceived
  (params (Instigator :type Controller%20Object%20Reference))
  ;[0.3018 0 0.5312] Receivers
  (Variables|Default|SetCachedInstigator Instigator)
  (CallFunction|RespawnVehicleFromSettings))

(event Custom|DestroyOnReceived
  (params (Instigator :type Controller%20Object%20Reference))
  ;[0.3018 0 0.5312] Receivers
  (Utilities|IsValid (Utilities|ToObject(Interface) (Variables|Default|GetCachedVehicleInterface))
    (:Is%20Valid
      (Variables|Default|SetCachedInstigator Instigator)
      (Utilities|FlowControl|Delay 0.0)
      (CallFunction|ManuallyDestroyVehicle)
      (CallFunction|VehicleDestroyed 0))
    (:Is%20Not%20Valid
      (Networking|SwitchHasAuthority
        (:Authority
          (CallFunction|AttemptRespawnVehicle))))) ; This check is done in cast the cached vehicle is pending kill
)

(event Custom|AssignDriverOnReceived
  (params (Instigator :type Controller%20Object%20Reference))
  ;[0.3018 0 0.5312] Receivers
  (Utilities|IsValid Instigator
    (:Is%20Valid
      (bind _asfort_player_controller (Utilities|Casting|CastToFortPlayerController Instigator))
      (bind _asfort_player_pawn (Utilities|Casting|CastToFortPlayerPawn (Pawn|GetControlledPawn _asfort_player_controller)))
      (bind _returnvalue (Minigame|GetMinigameforActor _asfort_player_pawn))
      (Utilities|IsValid _returnvalue
        (:Is%20Valid
          (if (Minigame|IsPostGame _returnvalue)
            _
            (elif (Variables|Default|GetbPendingReseat)
              (Variables|Default|SetPendingDriver _asfort_player_pawn)
              (else
                (Variables|Default|SetPawnToPlaceInVehicle _asfort_player_pawn)
                (CallFunction|AttemptSeatPawn)))))))))

(event Custom|On%20Pawn%20Entered%20Vehicle
  (params (Vehicle :type Fort%20Vehicle%20Interface%20Interface%20%28by%20ref%29) (PlayerPawn :type Fort%20Pawn%20Object%20Reference) (SeatIndex :type Integer))
  (bind _returnvalue (Pawn|GetController PlayerPawn))
  ;[0 0.5151 1 0.5] Transmitters
  (Utilities|IsValid PlayerPawn
    (:Is%20Valid
      (Utilities|IsValid _returnvalue
        (:Is%20Valid
          (GameplayEvent|BroadcastEvent (Variables|UserOptions|Events|GetOn%20Player%20Enters%20Vehicle) 0 (Utilities|Struct|MakeGameplayEventPlayerEnteredVehicle _returnvalue)))))))

(event Custom|On%20Pawn%20Exited%20Vehicle
  (params (Vehicle :type Fort%20Vehicle%20Interface%20Interface%20%28by%20ref%29) (PlayerPawn :type Fort%20Pawn%20Object%20Reference) (SeatIndex :type Integer))
  ;[0 0.5151 1 0.5] Transmitters
  (bind _asfort_player_pawn (Utilities|Casting|CastToFortPlayerPawn PlayerPawn))
  (Utilities|IsValid (Ability|GetFortAbilitySystemComponent _asfort_player_pawn)
    (:Is%20Valid
      (if (and (not (Vehicle|IsinVehicle _asfort_player_pawn)) (Abilities|HasActiveAbilitywithAnyMatchingTag (Ability|GetFortAbilitySystemComponent _asfort_player_pawn) (Variables|Default|GetInVehicleAbilityTags)))
        (Utilities|IsValid (Fortnite|GetFortPlayerControllerfromActor _asfort_player_pawn)
          (:Is%20Valid
            (Vehicle|BindEventtoOnVehicleAbilitiesRemoved (Fortnite|GetFortPlayerControllerfromActor _asfort_player_pawn) (delegate Custom|OnVehicleAbilitiesRemoved))))
        (else
          (Utilities|IsValid (Fortnite|GetFortPlayerControllerfromActor _asfort_player_pawn)
            (:Is%20Valid
              (if (not (Vehicle|IsinVehicle _asfort_player_pawn))
                (GameplayEvent|BroadcastEvent (Variables|UserOptions|Events|GetOn%20Player%20Exits%20Vehicle) 0 (Utilities|Struct|MakeGameplayEventPlayerExitedVehicle (Fortnite|GetFortPlayerControllerfromActor _asfort_player_pawn)))))))))))

(event Custom|OnVehicleAbilitiesRemoved
  (params (ExitedPlayerController :type Fort%20Player%20Controller%20Object%20Reference))
  ;[0 0.5151 1 0.5] Transmitters
  (bind _asfort_player_controller_athena (Utilities|Casting|CastToFortPlayerControllerAthena ExitedPlayerController)
    (:then
      (if (not (Class|FortPlayerControllerAthena|IsTeleportingtoLobby _asfort_player_controller_athena))
        (GameplayEvent|BroadcastEvent (Variables|UserOptions|Events|GetOn%20Player%20Exits%20Vehicle) 0 (Utilities|Struct|MakeGameplayEventPlayerExitedVehicle ExitedPlayerController)))
      (Vehicle|UnbindEventfromOnVehicleAbilitiesRemoved ExitedPlayerController (delegate Custom|OnVehicleAbilitiesRemoved)))
    (:CastFailed
      _)))

(event EventOnWorldReady
  ;[0.5541 1 0.6092 0.75] Handle Collision here based on visibility settings of the spawner
  (|CallParentFunction)
  (CallFunction|HandleEditModeCollision))

(event Custom|CheckEjectOnCooldown
  ;[1 0.5885 0.5885 0.8] Handle resetting the ejection cooldown if time has elapsed
  (Utilities|FlowControl|RetriggerableDelay (Variables|Default|GetEjectCooldownDelay))
  (Variables|Default|SetbEjectOnCooldown false))

(event Custom|CheckSeatingChannel
  ;[1 0.8379 0.211] When respawning on game start, set bool to attempt to seat the cahced pawn to seat if it exists after the vehicle respawns. The allows player spawn pads to seat a player on spawn even when the vehicle is in the process of respawning
  (Networking|SwitchHasAuthority
    (:Authority
      (Variables|Default|SetbPendingReseat true)
      (CallFunction|RefreshVehicleSpawn))))

(event Custom|CheckToSeatOnStart
  (bind _pendingdriver (Variables|Default|GetPendingDriver))
  ;[1 0.8379 0.211] When respawning on game start, set bool to attempt to seat the cahced pawn to seat if it exists after the vehicle respawns. The allows player spawn pads to seat a player on spawn even when the vehicle is in the process of respawning
  (Networking|SwitchHasAuthority
    (:Authority
      (if (Variables|Default|GetbPendingReseat)
        (Utilities|IsValid _pendingdriver
          (:Is%20Valid
            (CallFunction|SetDriveronStart _pendingdriver)
            (Variables|Default|SetPendingDriver 0))
          (:Is%20Not%20Valid
            _))
        (Variables|Default|SetbPendingReseat false)))))

(event Custom|SetDriverOnStart
  (params (NewDriver :type Fort%20Player%20Pawn%20Object%20Reference))
  ;[1 0.8379 0.211] When respawning on game start, set bool to attempt to seat the cahced pawn to seat if it exists after the vehicle respawns. The allows player spawn pads to seat a player on spawn even when the vehicle is in the process of respawning
  (Networking|SwitchHasAuthority
    (:Authority
      (Variables|Default|SetPawnToPlaceInVehicle NewDriver)
      (CallFunction|AttemptSeatPawn))))

(fn AttemptSeatPawn
  (bind _localathenavehicle (LocalVariables|GetLocalAthenaVehicle))
  (bind _returnvalue (Vehicle|GetDriver _localathenavehicle))
  (bind _pawntoplaceinvehicle (Variables|Default|GetPawnToPlaceInVehicle))
  (bind _localathenavehicle_2 (LocalVariables|GetLocalAthenaVehicle))
  ;[1 0.8379 0.211] Handle seating a pawn based on if there is a driver, and if there is what team the driver is on
  (Utilities|IsValid (Variables|Default|GetPawnToPlaceInVehicle)
    (:Is%20Valid
      (bind _asfort_athena_vehicle (Utilities|Casting|CastToFortAthenaVehicle (Variables|Default|GetCachedVehicleInterface)))
      (Utilities|IsValid _asfort_athena_vehicle
        (:Is%20Valid
          (if (or (Class|FortAthenaVehicle|IsPendingDeath _asfort_athena_vehicle) (<= (Class|FortVehicleInterface|GetHealth _asfort_athena_vehicle) 0.0))
            (Development|PrintString "Pending Death - Don't seat new pawn")
            (else
              (CallFunction|CancelAbilitieswithTags)
              (LocalVariables|SetLocalAthenaVehicle _asfort_athena_vehicle)
              (Utilities|FlowControl|Sequence
                (:then_0
                  (Utilities|IsValid _returnvalue
                    (:Is%20Valid
                      (if (Team|OnSameTeam _returnvalue (Variables|Default|GetPawnToPlaceInVehicle))
                        (bind _returnvalue_2 (Vehicle|ForceSeatSinglePawn _localathenavehicle_2 _pawntoplaceinvehicle -1))
                        (LocalVariables|SetbPawnSeated _returnvalue_2)
                        (else
                          (CallFunction|CheckEjectonCooldown)
                          (if (not (Variables|Default|GetbEjectOnCooldown))
                            (Vehicle|EjectSeat _localathenavehicle_2 0) ; Eject the current driver if they are on another team
                            (bind _returnvalue_3 (Vehicle|ForceSeatSinglePawn _localathenavehicle_2 _pawntoplaceinvehicle -1))
                            (LocalVariables|SetbPawnSeated _returnvalue_3)
                            (Variables|Default|SetbEjectOnCooldown true)))))
                    (:Is%20Not%20Valid
                      (bind _returnvalue_4 (Vehicle|ForceSeatSinglePawn (LocalVariables|GetLocalAthenaVehicle) (Variables|Default|GetPawnToPlaceInVehicle) -1))
                      (LocalVariables|SetbPawnSeated _returnvalue_4))) ; Is there currently a driver
                )
                (:then_1
                  (if (LocalVariables|GetbPawnSeated)
                    (CallFunction|OnPawnEnteredVehicle (Variables|Default|GetCachedVehicleInterface) (Variables|Default|GetPawnToPlaceInVehicle)) ; Our method of force seating the player does not\ntrigger the delegate we listen to, so we manually call it here
                    (else
                      _))
                  (Variables|Default|SetPawnToPlaceInVehicle 0) ; null global variable now that it has been used
                )) ; Null pawn to seat after attempting to seat
            )))) ; Cast can pass even if vehicle is destroyed \nso this will make sure the vehicle still exists
    )
    (:Is%20Not%20Valid
      (Development|PrintString "Failed to place pawn in vehicle. Pawn Invalid" true true "(R=1.000000,G=0.039115,B=0.000000,A=1.000000)"))))

(fn CancelAbilitiesWithTags
  (bind _pawntoplaceinvehicle (Variables|Default|GetPawnToPlaceInVehicle))
  (bind _returnvalue (Ability|GetAbilitySystemComponent _pawntoplaceinvehicle))
  (bind (_asfort_ability_system_component _bsuccess) (Utilities|Casting|CastToFortAbilitySystemComponent _returnvalue))
  (if _bsuccess
    (Tags|CancelAbilitiesWithTags _asfort_ability_system_component (LocalVariables|GetPossessedPropTag))))

(fn DestroySpawnedVehicle
  (bind _cachedvehicleinterface (Variables|Default|GetCachedVehicleInterface))
  (bind _returnvalue (Utilities|ToObject(Interface) _cachedvehicleinterface))
  ;[1 0.1543 0] Destroys the current vehicle if it exists
  (Utilities|IsValid _returnvalue
    (:Is%20Valid
      (bind _asfort_athena_vehicle (Utilities|Casting|CastToFortAthenaVehicle _returnvalue))
      (Vehicle|ForceDestroyVehicle _asfort_athena_vehicle false))) ; This check is done in cast the cached vehicle is pending kill
)

(fn GetCollisionStaticMesh
  (bind _staticmeshcomponent (Variables|Debug|GetStaticMeshComponent))
  (if (Utilities|IsValid _staticmeshcomponent :!class /Script/Engine.KismetSystemLibrary)
    (return "0, 0, 0" (Variables|/Script/Engine.StaticMeshComponent|GetStaticMesh _staticmeshcomponent))
    (else
      (return "0, 0, 0" 0))))

(fn GetCreativeActorBounds
  (bind _boundextentsoverride (Variables|Default|GetBoundExtentsOverride))
  (bind _boundstransform_matrix (Variables|GetBoundsTransform_Matrix))
  (bind (_xplane _yplane _zplane) (Utilities|Struct|BreakMatrix _boundstransform_matrix))
  (bind (_w _x _y _z) (Utilities|Struct|BreakPlane _zplane))
  (bind (_w_2 _x_2 _y_2 _z_2) (Utilities|Struct|BreakPlane _xplane))
  (bind (_w_3 _x_3 _y_3 _z_3) (Utilities|Struct|BreakPlane _yplane))
  (bind (_boundsorigin_x _boundsorigin_y _boundsorigin_z) (Math|Vector|BreakVector (Variables|GetBoundsOrigin_X)))
  (bind (_boundsextent_x _boundsextent_y _boundsextent_z) (Math|Vector|BreakVector (Variables|GetBoundsExtent_X)))
  ; GetInitialBoundsValues
  (Variables|SetBoundsMin (Math|Vector|NegateVector _boundextentsoverride))
  (Variables|SetBoundsMax _boundextentsoverride)
  (Variables|SetBoundsOrigin (* (+ (Variables|GetBoundsMax) (Variables|GetBoundsMin)) 0.5))
  (Variables|SetBoundsExtent (* (- (Variables|GetBoundsMax) (Variables|GetBoundsMin)) 0.5))
  ; GetTransformForBounds
  (Variables|SetBoundsTransform_Matrix (Math|Transform|ToMatrix(Transform) (Math|Transform|MakeTransform "0, 0, 0" (Transformation|GetActorRotation self) (Transformation|GetActorScale3D self))))
  (Variables|SetMatrix_XVector (Math|Vector|MakeVector _x_2 _y_2 _z_2)) ; Transform's X component
  (Variables|SetMatrix_YVector (Math|Vector|MakeVector _x_3 _y_3 _z_3)) ; Transform's Y component
  (Variables|SetMatrix_ZVector (Math|Vector|MakeVector _x _y _z)) ; Transform's Z component
  ; GetNewBoundsOrigin
  (bind _output_get (Variables|SetBoundsOrigin_New (Utilities|Operators|Add (* (Variables|GetMatrix_XVector) _boundsorigin_x) (* (Variables|GetMatrix_YVector) _boundsorigin_y) (* (Variables|GetMatrix_ZVector) _boundsorigin_z))))
  ; GetNewBoundsExtent
  (bind _output_get_2 (Variables|SetBoundsExtent_New (Math|Vector|VectorGetAbs (* (Variables|GetMatrix_XVector) _boundsextent_x))))
  (bind _output_get_3 (Variables|SetBoundsExtent_New (+ _output_get_2 (Math|Vector|VectorGetAbs (* (Variables|GetMatrix_YVector) _boundsextent_y)))))
  (bind _output_get_4 (Variables|SetBoundsExtent_New (+ _output_get_3 (Math|Vector|VectorGetAbs (* (Variables|GetMatrix_ZVector) _boundsextent_z)))))
  (return (Variables|Default|GetbOverrideBoundExtents) (+ _output_get_4 _output_get)))

(fn GetCreativeActorOrigin
  (return (Variables|Default|GetbOverrideBoundOrigin) (Variables|Default|GetBoundOriginOverride)))

(fn GetCreativeActorRelativeTransform
  (return false 0))

(fn GetGridAxisGuideAttachmentPoint
  (return false false "0, 0, 0"))

(fn GetMeshComponents
  (return 0))

(fn HandleEditorVisibility
  ;[0.8827 0.1953 1 0.25] If we're the editor, hide the base if we don't want to show.
  ; (since we don't need to interact with it and we have the preview)
  (if (Editor|InEditor)
    (Rendering|SetVisibility (Variables|Debug|GetStaticMeshComponent) (Variables|LegacyOptions|GetbIsVisible) false)))

(fn HandleEditModeCollision
  (if (not (CallFunction|IsinPlayMode))
    (bind _visibilitycomponents (CallFunction|GetVisibilityComponents))
    (for _array_element _visibilitycomponents
      (Utilities|IsValid _array_element
        (:Is%20Valid
          (bind _asstatic_mesh_component (Utilities|Casting|CastToStaticMeshComponent _array_element))
          (if (or (Building|IsPreviewBuildingActor) (Variables|LegacyOptions|GetbIsVisible))
            (Collision|SetCollisionProfileName _asstatic_mesh_component "FortBuildingMeshPhysics" true :!class /Script/Engine.PrimitiveComponent)
            (Rendering|Material|SetMaterial _asstatic_mesh_component 0 "/Game/Creative/Devices/VehicleSpawner/MI_Floor_Spawnpad_Creative_Vehicle_Inst.MI_Floor_Spawnpad_Creative_Vehicle_Inst")
            (Rendering|Material|SetMaterial _asstatic_mesh_component 1 "/Engine/EngineMaterials/WorldGridMaterial.WorldGridMaterial")
            (else
              (Collision|SetCollisionProfileName _asstatic_mesh_component "FortTriggerOnlyInteractions" true :!class /Script/Engine.PrimitiveComponent)
              (Rendering|Material|SetMaterial _asstatic_mesh_component 0 "/Game/Creative/Devices/Generic/Materials/M_CP_Device_Hologram_Vehicles.M_CP_Device_Hologram_Vehicles")
              (Rendering|Material|SetMaterial _asstatic_mesh_component 1 "/Game/Creative/Devices/Generic/Materials/M_CP_Device_Hologram_Vehicles.M_CP_Device_Hologram_Vehicles")
              (Collision|SetCollisionResponsetoChannel _asstatic_mesh_component "ECC_Visibility" "ECR_Overlap"))))))))

(fn HandleVehicleReferenceVisibility
  (params (IsForPreviewing :type Boolean))
  ;[0.6214 1 0.9501] Set visibility of the reference (placement only) for the spawner according to MoveTool usage
  (if IsForPreviewing
    ;[0.6436 0 1] Only spawn BP_VehiclePreview If can execute cosmetic events is true
    (Networking|CanExecuteCosmeticEvents
      (:True
        (Collision|SetCollisionEnabled (Variables|Default|GetBox) "QueryOnly")
        ;[0.6156 0.291 1] Spawn BP_VehicleSpawnPreview at location of Vehicle Reference
        (Game|SpawnActorfromClass "/CRD_VehicleSpawners/BP_VehicleSpawnerPreview.BP_VehicleSpawnerPreview_C" (Transformation|GetWorldTransform (Variables|Default|GetSpawnLocation)) "Undefined" "OverrideRootScale" self (Variables|Default|GetBox) (Variables|Default|GetPreviewVehicleMesh) 0)))))

(fn HasAdditionalValidPlacementLogic
  (return true))

(fn IgnoreWorldGridScaleOverride
  (return false))

(fn InitializeVehicleHealth
  (params (In%20Vehicle :type Fort%20Athena%20Vehicle%20Object%20Reference))
  (bind _maxvehiclehealth (Variables|UserKnobs|GetMaxVehicleHealth))
  (bind _returnvalue (> _maxvehiclehealth 0.0))
  (bind _in_vehicle (Variables|GetIn%20Vehicle))
  (bind _returnvalue_2 (Ability|GetFortAbilitySystemComponent _in_vehicle))
  ;[0.7716 1 0.6029] Override Health as set by User Options.  Don't worry about clearing the Effect, as changing the options results in the vehicle respawning anyway
  (Utilities|IsValid In%20Vehicle
    (:Is%20Valid
      (if (!= (Variables|UserKnobs|GetMaxVehicleHealth) 0.0)
        (Fortnite|SetCanbeDamaged (Variables|GetIn%20Vehicle) _returnvalue)
        (LocalVariables|SetHealth (Math|Float|SelectFloat (Variables|UserKnobs|GetMaxVehicleHealth) (Math|Conversions|ToFloat(Integer) (CallFunction|GetMaximumLegalReplicatedHealthValue)) _returnvalue))
        (Networking|SwitchHasAuthority
          (:Authority
            (Utilities|IsValid _returnvalue_2
              (:Is%20Valid
                (bind _returnvalue_3 (Ability|GameplayEffect|AssignTagSetbyCallerMagnitude (GameplayEffects|MakeOutgoingSpec :self _returnvalue_2 :GameplayEffectClass "/Game/Creative/Devices/Common/GameplayEffects/GE_Devices_Common_SetMaxHealth.GE_Devices_Common_SetMaxHealth_C" :Level 0.0) "(TagName=\"SetByCaller.MaxHealth\")" (LocalVariables|GetHealth)))
                (GameplayEffects|ApplyGameplayEffectSpecToSelf _returnvalue_2 _returnvalue_3)))))))))

(fn IsValidPlacement
  (return true))

(fn OnCooldown
  ;[0.9918 1 0.2351] Determine if the cooldown should be active. Works on both server and client by using the replicated Last Cooldown Start value
  (return (and (Class|CreativeEnabledComponent|IsEnabled (Variables|Default|GetEnabledComponent)) (or (Math|Boolean|ANDBoolean (> (+ (Variables|Default|GetLastCooldownStart) (Variables|UserKnobs|GetSpawnDelay)) (Fortnite|Sync|Time|GetServerWorldTimeSeconds)) (> (Variables|UserKnobs|GetSpawnDelay) 0.0) (> (Variables|Default|GetLastCooldownStart) 0.0)) (== (Variables|UserKnobs|GetSpawnDelay) -1.0)))))

(fn OnRep_LastCooldownStart
  ;[0.3417 1 0.9502] Start/Stop the cooldown flash effects
  (CallFunction|UpdateEffects))

(fn RefreshVehicleSpawn
  ;[0.2081 1 0.9801] Destroys the vehicle if it exists, or spawns one with no cooldown if there is no vehicle. Used when transitioning between edit and play mode
  (Networking|SwitchHasAuthority
    (:Authority
      (Utilities|Time|ClearandInvalidateTimerbyHandle (Variables|Default|GetVehicle%20Spawn%20Timer))
      (Variables|Default|SetInstant%20Respawn%20Override true)
      (Utilities|IsValid (Utilities|ToObject(Interface) (Variables|Default|GetCachedVehicleInterface))
        (:Is%20Valid
          (bind _asfort_athena_vehicle (Utilities|Casting|CastToFortAthenaVehicle (Variables|Default|GetCachedVehicleInterface)))
          (Vehicle|UnbindEventfromOnPawnExitedVehicleSeat _asfort_athena_vehicle (EventDispatchers|CreateEvent :!event On%20Pawn%20Exited%20Vehicle)) ; Make sure we don't transmit when exiting vehicle between states
          (CallFunction|DestroySpawnedVehicle))
        (:Is%20Not%20Valid
          (CallFunction|RespawnVehicleFromSettings))))))

(fn RespawnVehiclefromSettings
  ;[0.25 0.3348 1] Destroys the current vehicle and prepares a new one to be spawned if the device is enabled or Instant Respawn Override is set
  (Networking|SwitchHasAuthority
    (:Authority
      (if (Variables|Default|GetInstant%20Respawn%20Override)
        (Variables|Default|SetInstant%20Respawn%20Override false)
        (else
          _))
      (if (Class|CreativeEnabledComponent|IsEnabled (Variables|Default|GetEnabledComponent))
        (CallFunction|ManuallyDestroyVehicle)
        (bind _asfort_athena_vehicle (Utilities|Casting|CastToFortAthenaVehicle (Variables|Default|GetCachedVehicleInterface))
          (:then
            (Utilities|IsValid _asfort_athena_vehicle
              (:Is%20Valid
                (Game|BindEventtoOnEndPlay _asfort_athena_vehicle (EventDispatchers|CreateEvent :!event Throttle%20New%20Vehicle%20On%20Vehicle%20End%20Play)))
              (:Is%20Not%20Valid
                _)))
          (:CastFailed
            _))
        (CallFunction|ThrottleNewVehicle)))))

(fn ShouldBlockPlacement
  (params (MoveTool :type Fort%20Creative%20Move%20Tool%20Object%20Reference))
  (bind _returnvalue (CreativeMoveTool|GetCurrentHitTraceRule MoveTool))
  (switch Utilities|FlowControl|Switch|SwitchonEHitTraceRule _returnvalue
    (:Visibility
      _)
    (:Terrain
      _)
    (:None
      ;[1 1 1] Block placement for any overlapped device or volume without "ignorevehicleplacementblock" tag
      (return false)))
  (bind _outactors (Collision|ComponentOverlapActors (Variables|Default|GetBox) (Transformation|GetWorldTransform (Variables|Default|GetBox)) (Utilities|Select :Index _returnvalue :Visibility (Utilities|Array|MakeArray "ObjectTypeQuery1" "ObjectTypeQuery2" "ObjectTypeQuery3" "ObjectTypeQuery4" "ObjectTypeQuery5" "ObjectTypeQuery6") :Terrain (Utilities|Array|MakeArray "ObjectTypeQuery1") :None "ObjectTypeQuery1") 0 (Utilities|Array|MakeArray (Components|GetOwner (Variables|Default|GetBox)) self)))
  (bind _array_element (Utilities|Array|ForEachLoopwithBreak _outactors)
    (:LoopBody
      (return (not (Actor|ActorHasTag _array_element "ignorevehicleplacementblock"))))
    (:Completed)))

(fn SpawnPointInVolume
  (Utilities|IsLocationRelevanttoMe (Transformation|GetWorldLocation (Variables|Default|GetSpawnLocation))
    (:IsRelevant
      (bind _area_clear (CallFunction|TestVehicleOverlap))
      (if _area_clear
        (return true)
        (else
          _)))
    (:NotRelevant
      _))
  (return false))

(fn TestVehicleOverlap
  (bind (_origin _boxextent) (Collision|GetComponentBounds (Variables|Default|GetBox)))
  (Utilities|Array|Add (LocalVariables|GetIgnoreActors) self)
  (Utilities|IsValid (Utilities|ToObject(Interface) (Variables|Default|GetCachedVehicleInterface))
    (:Is%20Valid
      (bind _asactor (Utilities|Casting|CastToActor (Utilities|ToObject(Interface) (Variables|Default|GetCachedVehicleInterface)))
        (:then
          (Utilities|Array|Add (LocalVariables|GetIgnoreActors) _asactor))
        (:CastFailed
          _)))
    (:Is%20Not%20Valid
      _))
  (bind (_outactors _returnvalue) (Collision|BoxOverlapActors _origin _boxextent (Utilities|Array|MakeArray "ObjectTypeQuery5" "ObjectTypeQuery12") 0 (LocalVariables|GetIgnoreActors) :!class /Script/Engine.KismetSystemLibrary))
  (return (not _returnvalue)))

(fn ThrottleNewVehicle
  ;[0.4337 1 0.2604] Use the Creative Throttle Manager to throttle the spawn of a new vehicle
  (if (not (Variables|SpamProtection|GetWaiting%20For%20Throttle))
    (bind _returnvalue (Class|CreativeBlueprintLibrary|GetCreativeGeneralThrottleManager))
    (Utilities|IsValid _returnvalue
      (:Is%20Valid
        (Variables|SpamProtection|SetWaiting%20For%20Throttle true)
        (Class|FortCreativeGeneralThrottleManager|RequestThrottleObject _returnvalue self (Math|Conversions|ToInteger(Byte) (Utilities|Enum|LiteralenumEGeneralThrottleBuckets "NewEnumerator0")) (EventDispatchers|CreateEvent :!event Completed%20Throttle))
        (Variables|Default|SetLastCooldownStart 0.0)))))

(fn TransmitOnCleanup
  (params (Instigator :type Controller%20Object%20Reference))
  (GameplayEvent|BroadcastEvent (Variables|UserOptions|Events|GetOn%20Vehicle%20Is%20Destroyed) 0 (Utilities|Struct|MakeGameplayEventVehicleDestroyed Instigator)))

(fn UpdateEffects
  (Variables|Setlocal%20Cooldown (CallFunction|OnCooldown))
  (Networking|CanExecuteCosmeticEvents
    (:True
      (Utilities|FlowControl|Sequence
        (:then_0
          (if (CallFunction|IsinPlayMode)
            ;[0.1068 0.7917 0.05773] Setup cooldown effect material on the floor spawner mesh
            (bind _mid_spawnplate (Variables|Cosmetics|GetMID%20SpawnPlate)
              (:then
                (Rendering|Material|SetMaterial (Variables|Debug|GetStaticMeshComponent) 0 _mid_spawnplate))
              (:else
                (bind _returnvalue (Rendering|Material|CreateDynamicMaterialInstance (Variables|Debug|GetStaticMeshComponent) 0 "/Game/Creative/Devices/VehicleSpawner/MI_Floor_Spawnpad_Creative_Vehicle_Inst.MI_Floor_Spawnpad_Creative_Vehicle_Inst" "None" :!class /Script/Engine.PrimitiveComponent))
                (bind _output_get (Variables|Cosmetics|SetMID%20SpawnPlate _returnvalue))
                (Rendering|Material|SetTextureParameterValue _output_get "EmissiveTexture" (Variables|Cosmetics|GetEmissive%20Texture))))
            ;[0.8361 0.08416 1] Update the floor light material
            (bind _output_get_2 (LocalVariables|SetLocal%20Color (Math|Color|SelectColor (Math|Color|SelectColor (Variables|ColorSwitches|GetColorCooldown) (Variables|ColorSwitches|GetEnabledColor) (Variables|Getlocal%20Cooldown)) (Variables|ColorSwitches|GetDisabledColor) (Class|CreativeEnabledComponent|IsEnabled (Variables|Default|GetEnabledComponent)))))
            (CallFunction|BlendEmissiveColor _output_get_2) ; Interpolate the material color from the current state to the new state
            (Rendering|Material|SetScalarParameterValue (Variables|Cosmetics|GetMID%20SpawnPlate) "FlickerScale" (select (Variables|Getlocal%20Cooldown) (Variables|Cosmetics|GetSpawn%20Plate%20Flicker%20Scale) 0.0) :!class /Script/Engine.MaterialInstanceDynamic) ; Enable or disable a flash effect by changing the FlickerScale
          ))
        (:then_1
          ;[0.1085 0.9003 1] Play the ready sound
          (if (Class|CreativeEnabledComponent|IsEnabled (Variables|Default|GetEnabledComponent))
            (Audio|PlaySoundatLocation (Variables|Cosmetics|GetTrap%20Ready%20Sound) (Transformation|GetActorLocation))))))))

(fn UpdateVehicleConfigs
  (bind _teamsetting (Variables|UserKnobs|GetTeamSetting))
  (bind _knob_selectedclass (Variables|UserKnobs|GetKnob_SelectedClass))
  (bind _bhasinfinitefuel (Variables|Default|GetbHasInfiniteFuel))
  (bind _startingfuel (Variables|UserKnobs|GetStartingFuel))
  ;[0.4608 1 0.3476 0.4] Set the vehicle runtime settings based on the devices user options
  (bind _asfort_athena_vehicle (Utilities|Casting|CastToFortAthenaVehicle (Variables|Default|GetCachedVehicleInterface)))
  (Class|FortAthenaVehicle|SetEnforceTeamRestriction _asfort_athena_vehicle (and (not (Knobs|Team|IsTeamAny _teamsetting)) (CallFunction|IsinPlayMode)))
  (Class|FortAthenaVehicle|SetMinigameTeamID _asfort_athena_vehicle (select (Knobs|Team|IsValidTeamIndex _teamsetting) (Teams|IndextoTeam (Knobs|Team|GetTeamValueasIndex _teamsetting)) (Utilities|Enum|LiteralenumEFortTeam "Spectator")))
  (Class|FortAthenaVehicle|SetEnforceMinigameClassRestriction _asfort_athena_vehicle (and (not (Knobs|Class|IsClassAll _knob_selectedclass)) (CallFunction|IsinPlayMode)))
  (Class|FortAthenaVehicle|SetMinigameClassID _asfort_athena_vehicle _knob_selectedclass)
  (Class|FortAthenaVehicle|SetWaterDestructionDelay _asfort_athena_vehicle (Variables|UserKnobs|GetWaterDestructionDelay))
  (Class|FortAthenaVehicle|SetForceInfiniteFuel _asfort_athena_vehicle _bhasinfinitefuel) ; Fuel support still forthcoming, so force it to be infinite for now.
  (if (and (Utilities|IsValid (Vehicle|Fuel|GetVehicleFuelComponent _asfort_athena_vehicle) :!class /Script/Engine.KismetSystemLibrary) (not _bhasinfinitefuel))
    (if (>= _startingfuel 0)
      (Class|FuelComponent|SetFuel (Vehicle|Fuel|GetVehicleFuelComponent _asfort_athena_vehicle) (Utilities|Operators|Multiply (Math|Conversions|ToFloat(Integer) _startingfuel) 0.01 (Class|FuelComponent|GetFuelCapacity (Vehicle|Fuel|GetVehicleFuelComponent _asfort_athena_vehicle))))
      (else
        _))
    (Class|FuelComponent|SetFuelUsageInfo (Vehicle|Fuel|GetVehicleFuelComponent _asfort_athena_vehicle) (Utilities|Struct|MakeRuntimeFuelUsageInfo (* (Variables|UserKnobs|GetFuelUseMultiplier) (Utilities|Struct|BreakRuntimeFuelUsageInfo (Class|FuelComponent|GetFuelUsageInfo (Vehicle|Fuel|GetVehicleFuelComponent _asfort_athena_vehicle)))) (* (.FuelPerSecondDriving (Utilities|Struct|BreakRuntimeFuelUsageInfo (Class|FuelComponent|GetFuelUsageInfo (Vehicle|Fuel|GetVehicleFuelComponent _asfort_athena_vehicle)))) (Variables|UserKnobs|GetFuelUseMultiplier)) (select (> (Variables|Default|GetBoostFuelUseMultiplier) 0) (* (.FuelPerSecondBoosting (Utilities|Struct|BreakRuntimeFuelUsageInfo (Class|FuelComponent|GetFuelUsageInfo (Vehicle|Fuel|GetVehicleFuelComponent _asfort_athena_vehicle)))) (Variables|Default|GetBoostFuelUseMultiplier)) (* (.FuelPerSecondDriving (Utilities|Struct|BreakRuntimeFuelUsageInfo (Class|FuelComponent|GetFuelUsageInfo (Vehicle|Fuel|GetVehicleFuelComponent _asfort_athena_vehicle)))) (Variables|UserKnobs|GetFuelUseMultiplier)))))))

(fn UpdateVehicleSpawnSettings
  (params (EnableOrDisableReceiver :type Boolean))
  ;[0 1 0.3818] Respawn or destroy the vehicle when bEnabled changes or options are changed
  (Networking|SwitchHasAuthority
    (:Authority
      (Utilities|Time|ClearandInvalidateTimerbyHandle (Variables|Default|GetVehicle%20Spawn%20Timer))
      (if (Class|CreativeEnabledComponent|IsEnabled (Variables|Default|GetEnabledComponent))
        (if (Math|Boolean|ANDBoolean (or (== (Variables|UserKnobs|GetRespawnOnEnable) 2) (and (== (Variables|UserKnobs|GetRespawnOnEnable) 1) (not (Utilities|IsValid (Utilities|ToObject(Interface) (Variables|Default|GetCachedVehicleInterface)) :!class /Script/Engine.KismetSystemLibrary)))) (not (CallFunction|OnCooldown)) (CallFunction|SpawnPointinVolume))
          (CallFunction|RespawnVehicleFromSettings)
          (elif (Variables|GetEnableOrDisableReceiver)
            (Variables|Default|SetCachedInstigator 0)))
        (elif (Variables|UserKnobs|GetbDestroyOnDisable)
          (CallFunction|ManuallyDestroyVehicle)
          (if (Variables|GetEnableOrDisableReceiver)
            (Variables|Default|SetCachedVehicleInterface 0)
            (GameplayEvent|BroadcastEvent (Variables|UserOptions|Events|GetOn%20Vehicle%20Is%20Destroyed) 0 (Utilities|Struct|MakeGameplayEventVehicleDestroyed (Variables|Default|GetCachedInstigator)))
            (Variables|Default|SetCachedInstigator 0)))))
    (:Remote
      (CallFunction|UpdateEffects))))

(fn ConstructionScript
  (|CallParentFunction))

(fn VolumeClearanceCheck)
