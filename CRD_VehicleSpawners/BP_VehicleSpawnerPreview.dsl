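; Hologram preview actor for a Vehicle Spawner, showing the vehicle mesh tinted by whether the spawn space is blocked.
; On BeginPlay, cosmetic-capable instances create a dynamic hologram material, async load the preview mesh and apply the material to every slot.
; A self-rescheduling timer overlap-tests the collision box against world static objects and vehicles to set valid or invalid tint.
; The actor binds to its owner's OnDestroyed to remove itself, avoiding a parent-held reference.
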(event EventBeginPlay
  (bind _staticmesh (Variables|Default|GetStaticMesh))
  ;[0.4948 0.2294 1] Create Dynamic Material Instance and load preview on spawn
  (Networking|CanExecuteCosmeticEvents
    (:True
      (bind _returnvalue (Rendering|Material|CreateDynamicMaterialInstance (Variables|Default|GetStaticMesh) 0 "/Game/Creative/Devices/Generic/Materials/M_CP_Device_Hologram_Vehicles.M_CP_Device_Hologram_Vehicles" "None" :!class /Script/Engine.PrimitiveComponent))
      (Variables|Default|SetMaterialInstance _returnvalue)
      (CallFunction|AsyncLoadPreview)
      (Transformation|SetAbsolute _staticmesh false false true)
      (Transformation|SetWorldScale3D _staticmesh "1.000000,1.000000,1.000000"))))

(event Custom|Async%20Load%20Preview
  ;[1 0.6509 0 0.25] Asnyc Load Mesh from the owning Vehicle Spawner (defined during spawn)
  (bind _vehicle_spawn_preview (Variables|Default|GetVehicle%20Spawn%20Preview)
    (:then
      (bind _object (Utilities|AsyncLoadAsset _vehicle_spawn_preview)
        (:Completed
          (bind _asstatic_mesh (Utilities|Casting|CastToStaticMesh _object))
          (bind _returnvalue (Components|StaticMesh|SetStaticMesh (Variables|Default|GetStaticMesh) _asstatic_mesh))
          (if _returnvalue
            (Utilities|IsValid (Actor|GetOwner)
              (:Is%20Valid
                (for (_ _array_index) (Rendering|Material|GetMaterials (Variables|Default|GetStaticMesh))
                  (Rendering|Material|SetMaterial (Variables|Default|GetStaticMesh) _array_index (Variables|Default|GetMaterialInstance)))
                ;[0.8877 0.205 1] Call Function for checking overlap
                (CallFunction|SetMaterialBasedOnOverlap)
                ;[1 1 1] Destroy when Parent is Destroyed (avoids having to store reference on parent)
                (Game|BindEventtoOnDestroyed (Actor|GetOwner) (delegate Custom|Destroy%20on%20Parent%20Destruction)))
              (:Is%20Not%20Valid
                (Actor|DestroyActor)))))))
    (:else
      (Actor|DestroyActor))))

(event Custom|Destroy%20on%20Parent%20Destruction
  (params (DestroyedActor :type Actor%20Object%20Reference))
  ;[1 1 1] Destroy when Parent is Destroyed (avoids having to store reference on parent)
  (Actor|DestroyActor))

(event Custom|Set%20Material%20Based%20on%20Overlap
  ;[0.7176 0.3955 1] Check if Static Mesh overlaps World static objects or vehicles
  (Variables|Default|GetMaterialInstance
    (:then
      (bind _collision_box (Variables|Default|GetCollision%20Box)
        (:then
          (bind (_outactors _returnvalue) (Collision|ComponentOverlapActors _collision_box (Transformation|GetWorldTransform _collision_box) (Utilities|Array|MakeArray "ObjectTypeQuery1" "ObjectTypeQuery5") 0 (Utilities|Array|MakeArray (Components|GetOwner _collision_box) self)))
          ;[0.464 0.1676 1] Set DiffuseTint in Material Instance based on overlap
          (Rendering|Material|SetVectorParameterValue (Variables|Default|GetMaterialInstance) (Variables|Default|GetDiffuseTintParameter) (Math|Color|SelectColor (Variables|Default|GetInvalidSpawnColor) (Variables|Default|GetValidSpawnColor) _returnvalue) :!class /Script/Engine.MaterialInstanceDynamic)
          ;[0.506 0.05185 1] Set Event timer to loop function
          (Utilities|Time|SetTimerbyEvent (delegate Custom|Set%20Material%20Based%20on%20Overlap) (Variables|Default|GetEventLoopTime)))
        (:else)))
    (:else)))

(fn DestroyPreview
  (params (DestroyedActor :type Actor%20Object%20Reference))
  ;[1 0.115 0.821] Destroy Self
  (Actor|DestroyActor))

(fn ConstructionScript)
