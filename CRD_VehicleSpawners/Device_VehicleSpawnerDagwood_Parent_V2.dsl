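; Parent Creative vehicle spawner device for Dagwood cars, applying user options to each spawned vehicle.
; On spawn it strips the radio when disabled, applies tire and cow catcher mods on authority, toggles a self-destruct gameplay effect, and optionally forces color via child-overridden style index.
; UpdateVehicleConfigs pushes boost, damage, fade and explosion settings to the vehicle.
; Accessors expose fuel and boost options, while MigrateLegacyUserOptions converts legacy boost regen values.
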(event EventOnVehicleSpawned
  (params (Spawned%20Vehicle :type Fort%20Athena%20Vehicle%20Object%20Reference))
  ;[0.09063 0.04957 1] Vehicle Spawning Logic -- Accounting for Dagwood-specific options
  (|CallParentFunction Spawned%20Vehicle)
  (CallFunction|HandleVehicleRadio (and (Variables|UserOptions|Basic|GetbRadioEnabled) (ScalableFloat|GetValueasBool (Variables|Default|GetRadioOverride)))) ; disable radio
  (CallFunction|HandleModSettings)
  (CallFunction|HandleSelfDestruct (and (Variables|UserKnobs|GetbAllowSelfDestruct) (ScalableFloat|GetValueasBool (Variables|Default|GetAllowSelfDestructOverride))))
  (if (>= (ChildFunctions|ColorandStyleasAnInteger) 0)
    (CallFunction|CheckColorSettings)))

(fn CheckColorSettings
  (bind _colorandstyle (ChildFunctions|ColorandStyleasAnInteger))
  ;[0.1451 1 0.6491 0.75] Override random material determination and set a specific one based on the vehicle
  (bind _asvalet_basic_car_vehicle (Utilities|Casting|CastToValet_BasicCar_Vehicle (Variables|Default|GetCachedVehicleInterface)))
  (if (Variables|/Valet/BasicCar/Valet_BasicCar_Vehicle.Valet_BasicCar_Vehicle_C|GetIsBasicModdedCar _asvalet_basic_car_vehicle)
    (Variables|/Valet/BasicCar/Valet_BasicCar_Vehicle.Valet_BasicCar_Vehicle_C|SetRandomMatInt_BasicModdedCar _colorandstyle _asvalet_basic_car_vehicle)
    (else
      (Variables|/Valet/BasicCar/Valet_BasicCar_Vehicle.Valet_BasicCar_Vehicle_C|SetRandomMatInt _colorandstyle _asvalet_basic_car_vehicle))))

(fn ColorAndStyleAsAnInteger
  (return -1))

(fn GetBoostEnabled
  (return (Variables|UserOptions|GetBoost%20Enabled)))

(fn GetBoostFuelUseMultiplier
  (return (Variables|UserOptions|GetBoost%20Fuel%20Use)))

(fn GetConsumesFuel
  (return (Variables|UserOptions|Basic|GetFuelConsumption)))

(fn GetFuelUseMultiplier
  (return (Variables|UserOptions|Basic|GetFuelUseMultiplier)))

(fn GetMaxRandomStartingFuel
  (return (Variables|UserOptions|Advanced|GetMax%20Random%20Starting%20Fuel)))

(fn GetMinRandomStartingFuel
  (return (Variables|UserOptions|Advanced|GetMin%20Random%20Starting%20Fuel)))

(fn GetRandomStartingFuel
  (return (Variables|UserOptions|Basic|GetRandom%20Starting%20Fuel)))

(fn GetStartingFuelValue
  (return (Variables|UserOptions|Basic|GetStarting%20Fuel)))

(fn HandleCowCatcherSettings
  (bind _cowcatchermodtag (Variables|Default|GetCowCatcherModTag))
  ;[1 0.7311 0.6711 0.75] If Cow Catcher is enabled, then apply the Cow Catcher to the vehicle.
  (Networking|SwitchHasAuthority
    (:Authority
      (if (and (ScalableFloat|GetValueasBool (Variables|Default|GetCowCatcherOverride)) (Variables|UserOptions|GetSpawn%20With%20Cow%20Catcher))
        (bind _asfort_dagwood_vehicle (Utilities|Casting|CastToFortDagwoodVehicle (Variables|Default|GetCachedVehicleInterface)))
        (if (GameplayTags|IsGameplayTagValid _cowcatchermodtag)
          (Vehicle|Mod|ApplyVehicleMod _asfort_dagwood_vehicle _cowcatchermodtag 0))))))

(fn HandleModSettings
  (CallFunction|HandleTireSettings)
  (CallFunction|HandleCowCatcherSettings))

(fn HandleSelfDestruct
  (params (Allow :type Boolean))
  (bind _ge_allowselfdestruct (Variables|Default|GetGE_AllowSelfDestruct))
  (bind _asfort_dagwood_vehicle (Utilities|Casting|CastToFortDagwoodVehicle (Variables|Default|GetCachedVehicleInterface)))
  (if Allow
    (GameplayEffects|ApplyGameplayEffectToSelf :self (Ability|GetAbilitySystemComponent _asfort_dagwood_vehicle) :GameplayEffectClass _ge_allowselfdestruct :Level 0.0)
    (else
      (GameplayEffects|RemoveActiveGameplayEffectbySourceEffect (Ability|GetAbilitySystemComponent _asfort_dagwood_vehicle) _ge_allowselfdestruct 0 -1))))

(fn HandleTireSettings
  (bind _tiremodtag_0 (Variables|Default|GetTireModTag_0))
  ;[1 0.7311 0.6711 0.75] If our tires are not Default (0), then set the tires to the proper mod
  (Networking|SwitchHasAuthority
    (:Authority
      (if (ScalableFloat|GetValueasBool (Variables|Default|GetTireOverride))
        (switch Utilities|FlowControl|Switch|SwitchonEVehicleTireModOption (Variables|UserOptions|GetTire%20Selection)
          (:NewEnumerator1
            (bind _asfort_dagwood_vehicle (Utilities|Casting|CastToFortDagwoodVehicle (Variables|Default|GetCachedVehicleInterface)))
            (if (GameplayTags|IsGameplayTagValid _tiremodtag_0)
              (Vehicle|Mod|ApplyVehicleMod _asfort_dagwood_vehicle _tiremodtag_0 0) ; Used instead of ApplyTireMod for proper internal setup
            )))))))

(fn HandleVehicleRadio
  (params (bRadioIsEnabled :type Boolean))
  ;[0.9583 0.14 0.04445] Handle Radio by setting.  At this time, we only destroy them, but we are keeping this in place in the event we want to add a radio component to non-Valet-derived vehicles in the future.
  (if (not bRadioIsEnabled)
    (bind _asvalet_basic_car_vehicle (Utilities|Casting|CastToValet_BasicCar_Vehicle (Variables|Default|GetCachedVehicleInterface)))
    (Utilities|IsValid (Actor|GetComponentbyClass _asvalet_basic_car_vehicle "/Script/SrirachaRanch.StreamingRadioPlayerComponent")
      (:Is%20Valid
        (Components|DestroyComponent (Actor|GetComponentbyClass _asvalet_basic_car_vehicle "/Script/SrirachaRanch.StreamingRadioPlayerComponent"))))))

(fn MigrateLegacyUserOptions
  (bind (_tiremodtag_0 _fuelusemultiplier _boostregensetting) (Class|GetClassDefaults "/CRD_VehicleSpawners/Device_VehicleSpawnerDagwood_Parent_V2.Device_VehicleSpawnerDagwood_Parent_V2_C"))
  (bind _boost_regen_multiplier (Variables|Legacy|GetBoost%20Regen%20Multiplier))
  (|CallParentFunction)
  (Utilities|FlowControl|Sequence
    (:then_0
      (Utilities|FlowControl|Sequence
        (:then_0
          (Utilities|FlowControl|Sequence
            (:then_0
              (if (!= (Variables|Legacy|GetBoostRegenSetting) _boostregensetting)
                (if (== (Variables|Legacy|GetBoostRegenSetting) 1.0)
                  (Variables|UserOptions|SetUnlimited%20Boost true)
                  (else
                    (bind _output_get (Variables|Legacy|SetBoost%20Regen%20Multiplier (Math|Float|MapRangeClamped (Variables|Legacy|GetBoostRegenSetting) 0.0 0.1 0.0 2.0)))))
                (Variables|UserOptions|SetBoost%20Enabled (> _output_get 0))
                (Variables|Legacy|SetBoostRegenSetting _boostregensetting)
                (LocalVariables|SetOptionMigrated true)))
            (:then_1
              (if (!= (Class|GetClassDefaults "/CRD_VehicleSpawners/Device_VehicleSpawnerDagwood_Parent_V2.Device_VehicleSpawnerDagwood_Parent_V2_C") _boost_regen_multiplier)
                (Variables|UserOptions|SetBoostRegenMultiplier (* _boost_regen_multiplier 1e+01))
                (Variables|Legacy|SetBoost%20Regen%20Multiplier _boost_regen_multiplier) ; Legacy
                (LocalVariables|SetOptionMigrated true)))))))
    (:then_1
      (if (LocalVariables|GetOptionMigrated)
        (OptionData|RefreshPropertyOverrides (Variables|Default|GetToyOptionsComponent))))))

(fn SetBoostFuelUseMultiplier
  (params (In%20Value :type Float%20%28double-precision%29))
  (Variables|UserOptions|SetBoost%20Fuel%20Use In%20Value))

(fn SetConsumeFuel
  (params (ConsumeFuel :type Boolean))
  (Variables|UserOptions|Basic|SetFuelConsumption ConsumeFuel))

(fn SetRandomStartingFuel
  (params (RandomStartingFuel :type Boolean))
  (Variables|UserOptions|Basic|SetRandom%20Starting%20Fuel RandomStartingFuel))

(fn SetStartingFuelValue
  (params (Starting%20Fuel%20Value :type Integer))
  (Variables|UserOptions|Basic|SetStarting%20Fuel Starting%20Fuel%20Value))

(fn SetBoostEnabled
  (params (BoostEnabled :type Boolean))
  (Variables|UserOptions|SetBoost%20Enabled BoostEnabled))

(fn SetMaxRandomStartingFuel
  (params (Max%20Fuel%20Percent :type Integer))
  (|CallParentFunction Max%20Fuel%20Percent)
  (Variables|UserOptions|Advanced|SetMax%20Random%20Starting%20Fuel Max%20Fuel%20Percent))

(fn SetMinRandomStartingFuel
  (params (Min%20Fuel%20Percent :type Integer))
  (|CallParentFunction Min%20Fuel%20Percent)
  (Variables|UserOptions|Advanced|SetMin%20Random%20Starting%20Fuel Min%20Fuel%20Percent))

(fn UpdateVehicleConfigs
  ;[0.4608 1 0.3476] Set the vehicle runtime settings based on the devices user options
  (|CallParentFunction)
  (bind _asfort_dagwood_vehicle (Utilities|Casting|CastToFortDagwoodVehicle (Variables|Default|GetCachedVehicleInterface)))
  (Class|FortDagwoodVehicle|SetRuntimeBoostInfoonServer _asfort_dagwood_vehicle (Utilities|Struct|MakeRuntimeBoostInfo (Variables|UserOptions|GetBoost%20Enabled) (not (Variables|UserOptions|Basic|GetFuelConsumption)) (Variables|Default|GetBoost%20Push%20Force) (Variables|Default|GetBoost%20Top%20Speed%20Force%20Multiplier) (Variables|Default|GetBoost%20Top%20Speed%20Multiplier) (Math|Float|SelectFloat 1.0 (* (* (Variables|UserOptions|GetBoostRegenMultiplier) 0.1) 0.05) (Variables|UserOptions|GetUnlimited%20Boost)) (Math|Float|SelectFloat 0.0 (Variables|Default|GetBoost%20Rate%20of%20Use) (Variables|UserOptions|GetUnlimited%20Boost))))
  (Class|FortDagwoodVehicle|SetDamageInfoonServer _asfort_dagwood_vehicle (Variables|UserOptions|GetbDamageFriendlyFire) (Variables|UserOptions|GetbDamageOtherVehicles) (Variables|UserOptions|GetbDamageOwnVehicle) (Variables|UserOptions|GetAllow%20Damage%20From%20Other%20Vehicles))
  ;;[1 0.9655 0.1244 1 12] Disable fade out by setting the duration to -1. Fade out is not necessary for Creative and makes some of the user settings confusing (MaxTimeExplosion and MaxTimeZeroHealth).
  (NativeFade|ServerSetFadeDuration _asfort_dagwood_vehicle -1.0)
  (Class|FortAthenaVehicle|ServerSetDestructionInfo _asfort_dagwood_vehicle (Variables|Default|GetSelfDestructTime) (Variables|UserOptions|GetLifetimeAfterExplosion) (Variables|Default|GetSelfDestructCleanupTime) (Variables|UserOptions|GetMax%20Explosion%20Delay))
  (bind _asvalet_basic_car_vehicle (Utilities|Casting|CastToValet_BasicCar_Vehicle _asfort_dagwood_vehicle))
  (Utilities|IsValid _asvalet_basic_car_vehicle
    (:Is%20Valid
      (Variables|/Valet/BasicCar/Valet_BasicCar_Vehicle.Valet_BasicCar_Vehicle_C|SetCachedExplosionDamageBuildings (Variables|UserOptions|GetExplosionDamageToEnvironment) _asvalet_basic_car_vehicle)
      (Variables|/Valet/BasicCar/Valet_BasicCar_Vehicle.Valet_BasicCar_Vehicle_C|SetCachedExplosionDamageVehicles (Variables|UserOptions|GetExplosionDamageToVehicles) _asvalet_basic_car_vehicle)
      (Variables|/Valet/BasicCar/Valet_BasicCar_Vehicle.Valet_BasicCar_Vehicle_C|SetCachedExplosionDamagePlayers (Variables|UserOptions|GetExplosionDamageToPlayers) _asvalet_basic_car_vehicle))))

(fn ConstructionScript
  (|CallParentFunction))
