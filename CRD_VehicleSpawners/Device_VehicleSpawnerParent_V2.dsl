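; Parent Blueprint for Creative vehicle spawner devices, which spawn, respawn and destroy a vehicle according to minigame phase and enabled state.
; The server spawns a transient vehicle and binds its delegates, using randomized spam-protection delays and a timer for respawns.
; Receivers toggle, respawn, repair or seat drivers, deferring seating during game-start respawns, while gameplay events report vehicle and occupant changes.
; Overridable functions supply per-vehicle settings, and preview meshes stand in while editing.
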(event EventBeginPlay
  ;[0.1083 1 0.0636] Initialization
  (Gameplay|AddReceiver (GameInstanceSubsystems|GetGameFrameworkComponentManager) 0 true)
  (CallFunction|InitializeDataRegistryResolverScope))

(event EventInitializeDevice
  (bind _returnvalue (Building|IsPreviewBuildingActor))
  ;[0.1083 1 0.0636] Initialization
  (CallFunction|InitializeDataRegistryOptions)
  (|CallParentFunction)
  (CallFunction|HandleEditModeCollision)
  (CallFunction|HandleEditorVisibility)
  (if _returnvalue
    (CallFunction|HandleVehicleReferenceVisibility _returnvalue)
    (else
      (CallFunction|HandleVehicleReferenceVisibility _returnvalue)
      (Class|CreativeEnabledComponent|SetEnabledOnPhase (Variables|Default|GetEnabledComponent) "PreGame" (ChildFunctions|GetEnabledStateasIndex))
      (CallFunction|UpdateVehicleSpawnSettings false)
      ;[0.8827 0.1953 1 0.25] We also want a preview in editor, but it can't spawn, so use a mesh.
      (if (Math|Boolean|ANDBoolean (Editor|InEditor) (Variables|Default|GetUseEditorPreview) (Utilities|IsValidSoftObjectReference (Variables|Default|GetPreviewVehicleMesh)))
        (bind _loaded_object (Creative|Loading|AsyncLoadEditorSafe (Variables|Default|GetPreviewVehicleMesh))
          (:Loaded
            (bind _asstatic_mesh (Utilities|Casting|CastToStaticMesh _loaded_object))
            (Components|StaticMesh|SetStaticMesh (Variables|Default|GetEditorVehiclePreview) _asstatic_mesh)))
        (else
          (Components|StaticMesh|SetStaticMesh (Variables|Default|GetEditorVehiclePreview) 0))))))

(event EventEndPlay
  (params (EndPlayReason :type EEndPlayReason%20Enum :default "Destroyed"))
  (bind _cachedvehicleinterface (Variables|Default|GetCachedVehicleInterface))
  ;[1 0.1508 0] Destruction
  (Variables|Default|SetCachedInstigator 0)
  (Networking|SwitchHasAuthority
    (:Authority
      (Utilities|IsValid (Utilities|ToObject(Interface) _cachedvehicleinterface)
        (:Is%20Valid
          (if (<= (Utilities|Array|Length (Vehicle|GetAllPassengers _cachedvehicleinterface)) 0)
            (CallFunction|ManuallyDestroyVehicle)
            (else
              (Vehicle|SetDestroyonLastExit (Variables|Default|GetCachedVehicleInterface) true))))))))

(event EventDestroyed
  (bind _cachedvehicleinterface (Variables|Default|GetCachedVehicleInterface))
  ;[1 0.1508 0] Destruction
  (Networking|SwitchHasAuthority
    (:Authority
      (Utilities|IsValid (Utilities|ToObject(Interface) _cachedvehicleinterface)
        (:Is%20Valid
          (if (<= (Utilities|Array|Length (Vehicle|GetAllPassengers _cachedvehicleinterface)) 0)
            (CallFunction|ManuallyDestroyVehicle)
            (else
              (Vehicle|SetDestroyonLastExit (Variables|Default|GetCachedVehicleInterface) true))))))))

(event Custom|Blend%20Emissive%20Color
  (params (BlendColor :type Linear%20Color%20Structure))
  ;[0.6287 0 1] Blend the color of the spawn pad lights when switching between states (Disabled, enabled, or on cooldown)
  (if (Math|Color|NotEqual(LinearColor) BlendColor (Variables|ColorSwitches|GetCurrent%20Color))
    (Variables|ColorSwitches|SetOld%20Color (Variables|ColorSwitches|GetCurrent%20Color))
    (Variables|ColorSwitches|SetCurrent%20Color BlendColor)
    (bind (_direction _alpha) (Timeline|Color%20Switch%20Timeline PlayFromStart)
      (:Update
        (Rendering|Material|SetVectorParameterValue (Variables|Cosmetics|GetMID%20SpawnPlate) "Emissive" (Math|Color|Lerp(LinearColor) (Variables|ColorSwitches|GetOld%20Color) (Variables|ColorSwitches|GetCurrent%20Color) _alpha) :!class /Script/Engine.MaterialInstanceDynamic)))))

(event MinigameStateChangedDelegate(FortMinigameProgress)
  (params (Minigame :type Fort%20Minigame%20Object%20Reference) (NewMinigameState :type EFortMinigameState%20Enum :default "PreGame"))
  ;[0.04638 0.9003 1 0.5] Handle Minigame State Change
  (Class|CreativeEnabledComponent|SetEnabledOnPhase (Variables|Default|GetEnabledComponent) NewMinigameState (ChildFunctions|GetEnabledStateasIndex))
  (if (Class|CreativeEnabledComponent|IsEnabled (Variables|Default|GetEnabledComponent))
    (switch Utilities|FlowControl|Switch|SwitchonEFortMinigameState NewMinigameState
      (:PreGame
        (CallFunction|UpdateVehicleSpawnSettings false)
        (CallFunction|RefreshVisibility))
      (:Setup
        _)
      (:Warmup
        _)
      (:InProgress
        (CallFunction|CheckSeatingChannel))
      (:PostRoundEnd
        _)
      (:PostGameReset
        _)
      (:Teardown
        (CallFunction|RefreshVisibility)))
    (CallFunction|RefreshVehicleSpawn)
    (else
      (CallFunction|ManuallyDestroyVehicle))))

(event Custom|Completed%20Throttle
  (params (Succeeded :type Boolean))
  ;[0.09063 0.04957 1] Vehicle Spawning Logic
  (Variables|SpamProtection|SetWaiting%20For%20Throttle false)
  (if Succeeded
    (bind _area_clear (CallFunction|TestVehicleOverlap))
    (if _area_clear
      (CallFunction|AttemptSpawnVehicle) ; If the vehicle has already been throttled, try to spawn it now
      (else
        _))
    (else
      _))
  (Variables|SpamProtection|SetWaiting%20to%20Spawn true)
  (CallFunction|AttemptSpawnAfterSpamProtectionDelay) ; If the vehicle failed to throttled, try to spawn it with delay to avoid spam spawning
)

(event Custom|Attempt%20Spawn%20Vehicle
  ;[0.09063 0.04957 1] Vehicle Spawning Logic
  (if (Variables|SpamProtection|GetRecently%20Spawned)
    (Variables|SpamProtection|SetWaiting%20to%20Spawn true)
    (else
      (CallFunction|SpawnVehicle))))

(event Custom|Attempt%20Spawn%20After%20Spam%20Protection%20Delay
  (bind (_spam_protection_min_max_time_x _spam_protection_min_max_time_y) (Math|Vector2D|BreakVector2D (Variables|GetSpam%20Protection%20Min/Max%20Time_X)))
  ;[0.09063 0.04957 1] Vehicle Spawning Logic
  ;;[1 0.984 0.1121 1 14] Waits a defined time period after spawning a vehicle, and then checks to see if anything has attempted to spawn a new one.
  ;; This is a buffer on the local device, stopping you mass-respawning vehicles.
  (Utilities|FlowControl|Delay (Math|Random|RandomFloatinRange _spam_protection_min_max_time_x _spam_protection_min_max_time_y))
  (Variables|SpamProtection|SetRecently%20Spawned false)
  (if (Variables|SpamProtection|GetWaiting%20to%20Spawn)
    (CallFunction|RespawnVehicleFromSettings)))

(event Custom|SpawnVehicle
  ;[0.09063 0.04957 1] Vehicle Spawning Logic
  (Networking|SwitchHasAuthority
    (:Authority
      (if (Math|Boolean|ANDBoolean (Fortnite|AreAthenaVehiclesEnabled) (Utilities|IsValidClass (Variables|Default|GetVehicleType)) (not (Editor|InEditor)))
        (bind _spawnedvehicle (Fortnite|Utilities|SpawnTransientVehicle (Variables|Default|GetVehicleType) (Variables|Default|GetPhysics%20Vehicle%20Config%20Class) (Math|Transform|MakeTransform (.location (Transformation|GetWorldTransform (Variables|Default|GetSpawnLocation))) (.rotation (Transformation|GetWorldTransform (Variables|Default|GetSpawnLocation)))) "AlwaysSpawn" self))
        (bind _asfort_vehicle_interface (Utilities|Casting|CastToFortVehicleInterface _spawnedvehicle))
        (Variables|Default|SetCachedVehicleInterface _asfort_vehicle_interface)
        (CallFunction|InitializeVehicleHealth _spawnedvehicle)
        (Default|InitializeVehicleVisual _spawnedvehicle)
        (InteractInterface|SetAllowInteractionBetweenFortVolumes _spawnedvehicle false :!class /Script/FortniteGame.FortAthenaVehicle)
        (CallFunction|OnVehicleSpawned _spawnedvehicle)))))

(event Custom|On%20Vehicle%20Spawned
  (params (Spawned%20Vehicle :type Fort%20Athena%20Vehicle%20Object%20Reference))
  (bind _returnvalue (Actor|GetComponentbyClass Spawned%20Vehicle "/Script/VehicleCosmeticsRuntime.VehicleCosmeticsAssembledMeshUserComponent"))
  ;[0.09063 0.04957 1] Vehicle Spawning Logic
  (Utilities|IsValid Spawned%20Vehicle
    (:Is%20Valid
      (Utilities|FlowControl|Sequence
        (:then_0
          (Variables|SpamProtection|SetWaiting%20to%20Spawn false)
          (Utilities|Time|ClearandInvalidateTimerbyHandle (Variables|Default|GetVehicle%20Spawn%20Timer))
          (Variables|SpamProtection|SetRecently%20Spawned true)
          (Vehicle|WakeUp Spawned%20Vehicle)
          (CallFunction|UpdateVehicleConfigs)
          (CallFunction|AttemptSpawnAfterSpamProtectionDelay))
        (:then_1
          (Game|BindEventtoOnEndPlay Spawned%20Vehicle (EventDispatchers|CreateEvent :!event Custom|Vehicle%20Destroyed))
          (Vehicle|AssignOnPawnFinishEnteringVehicle Spawned%20Vehicle (EventDispatchers|CreateEvent :!event Custom|On%20Pawn%20Entered%20Vehicle))
          (Vehicle|AssignOnPawnExitedVehicleSeat Spawned%20Vehicle (EventDispatchers|CreateEvent :!event Custom|On%20Pawn%20Exited%20Vehicle))
          (Health|BindEventtoOnHandleOutOfHealthDelegate Spawned%20Vehicle (EventDispatchers|CreateEvent :!event Custom|VehicleOutOfHealth)))
        (:then_2
          (GameplayEvent|BroadcastEvent (Variables|UserOptions|Events|GetOn%20Vehicle%20Spawns) 0 (Utilities|Struct|MakeGameplayEventVehicleSpawned (Variables|Default|GetCachedInstigator) Spawned%20Vehicle))
          (Variables|Default|SetCachedInstigator 0)
          (bind _asfort_athena_vehicle (Utilities|Casting|CastToFortAthenaVehicle (Variables|Default|GetCachedVehicleInterface)))
          (Class|FortAthenaVehicle|SetSupportCosmeticWrap _asfort_athena_vehicle (ChildFunctions|GetSupportsWraps)) ; Function has build in check that determines \nwhether the vehicle type should support wraps
          (CallFunction|ChecktoSeatonStart)
          (Utilities|IsValid _returnvalue
            (:Is%20Valid
              (Class|VehicleCosmeticsAMUC|SetVehicleCosmeticsEnabled _returnvalue (ScalableFloat|GetValueasBoolwithContext (Variables|Default|GetEnableVehicleCosmetics) 1.0)))))))))

(event Custom|Manually%20Destroy%20Vehicle
  (bind _cachedvehicleinterface (Variables|Default|GetCachedVehicleInterface))
  (bind _returnvalue (Utilities|ToObject(Interface) _cachedvehicleinterface))
  ;[1 0.3476 0] Vehicle Destroying Logic
  (Networking|SwitchHasAuthority
    (:Authority
      (Utilities|IsValid _returnvalue
        (:Is%20Valid
          (bind _asfort_athena_vehicle (Utilities|Casting|CastToFortAthenaVehicle _returnvalue))
          (CallFunction|RemoveVehicleBinds _asfort_athena_vehicle)
          (CallFunction|DestroySpawnedVehicle))) ; This check is done in cast the cached vehicle is pending kill
    )))

(event Custom|Remove%20Vehicle%20Binds
  (params (Vehicle :type Fort%20Athena%20Vehicle%20Object%20Reference))
  ;[1 0.3476 0] Vehicle Destroying Logic
  (Utilities|IsValid Vehicle
    (:Is%20Valid
      (Game|UnbindEventfromOnEndPlay Vehicle (delegate Custom|Vehicle%20Destroyed))
      (Vehicle|UnbindEventfromOnPawnFinishEnteringVehicle Vehicle (EventDispatchers|CreateEvent :!event Custom|On%20Pawn%20Entered%20Vehicle))
      (Vehicle|UnbindEventfromOnPawnExitedVehicleSeat Vehicle (EventDispatchers|CreateEvent :!event Custom|On%20Pawn%20Exited%20Vehicle)))))

(event Custom|Vehicle%20Destroyed
  (params (Actor :type Actor%20Object%20Reference) (EndPlayReason :type EEndPlayReason%20Enum :default "Destroyed"))
  ;[1 0.3476 0] Vehicle Destroying Logic
  (Networking|SwitchHasAuthority
    (:Authority
      (Variables|Default|SetCachedVehicleInterface 0)
      (GameplayEvent|BroadcastEvent (Variables|UserOptions|Events|GetOn%20Vehicle%20Is%20Destroyed) 0 (Utilities|Struct|MakeGameplayEventVehicleDestroyed (Variables|Default|GetCachedInstigator)))
      (Variables|Default|SetCachedInstigator 0)
      (CallFunction|AttemptRespawnVehicle))))

(event Custom|VehicleOutOfHealth
  (params (DeadVehicle :type Fort%20Athena%20Vehicle%20Object%20Reference) (InTags :type Gameplay%20Tag%20Container%20Structure%20%28by%20ref%29) (EventInstigator :type Controller%20Object%20Reference) (DamageCauser :type Actor%20Object%20Reference))
  ;[1 0.3476 0] Vehicle Destroying Logic
  (Networking|SwitchHasAuthority
    (:Authority
      (Utilities|IsValid EventInstigator
        (:Is%20Valid
          (Utilities|IsValid (Variables|Default|GetCachedInstigator)
            (:Is%20Not%20Valid
              (Variables|Default|SetCachedInstigator EventInstigator))))))))

(event Custom|Attempt%20Respawn%20Vehicle
  (bind _fortminigameprogress (Variables|Default|GetFortMinigameProgress))
  (bind _play_mode (Class|CreativeCommonDeviceFunctionLibrary|IsLogicinPlayMode _fortminigameprogress))
  (bind _returnvalue (not _play_mode))
  (bind _enablerespawn (Variables|UserOptions|GetEnableRespawn))
  ;[1 0.3476 0] Vehicle Destroying Logic
  (if (Math|Boolean|ORBoolean (and (== (Variables|UserOptions|GetRespawn%20Time) 0.0) _enablerespawn) (Variables|Default|GetInstant%20Respawn%20Override) (and (== (Variables|UserOptions|GetRespawn%20Time) 0.0) _returnvalue))
    (CallFunction|RespawnVehicleFromSettings)
    (elif (or _enablerespawn _returnvalue)
      (bind _returnvalue_2 (Utilities|Time|SetTimerbyFunctionName 0 "Respawn Vehicle From Settings" (Variables|UserOptions|GetRespawn%20Time)))
      (Variables|Default|SetVehicle%20Spawn%20Timer _returnvalue_2)
      (Variables|Default|SetLastCooldownStart (Fortnite|Sync|Time|GetServerWorldTimeSeconds)))))

(event Custom|Throttle%20New%20Vehicle%20On%20Vehicle%20End%20Play
  (params (Actor :type Actor%20Object%20Reference) (End%20Play%20Reason :type EEndPlayReason%20Enum :default "Destroyed"))
  ;[1 0.3476 0] Vehicle Destroying Logic
  (CallFunction|ThrottleNewVehicle))

(event Custom|On%20Pawn%20Entered%20Vehicle
  (params (Vehicle :type Fort%20Vehicle%20Interface%20Interface%20%28by%20ref%29) (PlayerPawn :type Fort%20Pawn%20Object%20Reference) (SeatIndex :type Integer))
  (bind _returnvalue (Pawn|GetController PlayerPawn))
  ;[0 0.5151 1 0.5] Transmitters
  (Utilities|IsValid PlayerPawn
    (:Is%20Valid
      (Utilities|IsValid _returnvalue
        (:Is%20Valid
          (GameplayEvent|BroadcastEvent (Variables|UserOptions|Events|GetOn%20Player%20Enters%20Vehicle) 0 (Utilities|Struct|MakeGameplayEventPlayerEnteredVehicle _returnvalue)))))))

(event Custom|On%20Pawn%20Exited%20Vehicle
  (params (Vehicle :type Fort%20Vehicle%20Interface%20Interface%20%28by%20ref%29) (PlayerPawn :type Fort%20Pawn%20Object%20Reference) (SeatIndex :type Integer))
  ;[0 0.5151 1 0.5] Transmitters
  (bind _asfort_player_pawn (Utilities|Casting|CastToFortPlayerPawn PlayerPawn))
  (Utilities|IsValid (Ability|GetFortAbilitySystemComponent _asfort_player_pawn)
    (:Is%20Valid
      (if (and (not (Vehicle|IsinVehicle _asfort_player_pawn)) (Abilities|HasActiveAbilitywithAnyMatchingTag (Ability|GetFortAbilitySystemComponent _asfort_player_pawn) (Variables|Default|GetInVehicleAbilityTags)))
        (Utilities|IsValid (Fortnite|GetFortPlayerControllerfromActor _asfort_player_pawn)
          (:Is%20Valid
            (Vehicle|BindEventtoOnVehicleAbilitiesRemoved (Fortnite|GetFortPlayerControllerfromActor _asfort_player_pawn) (delegate Custom|OnVehicleAbilitiesRemoved))))
        (else
          (Utilities|IsValid (Fortnite|GetFortPlayerControllerfromActor _asfort_player_pawn)
            (:Is%20Valid
              (if (not (Vehicle|IsinVehicle _asfort_player_pawn))
                (GameplayEvent|BroadcastEvent (Variables|UserOptions|Events|GetOn%20Player%20Exits%20Vehicle) 0 (Utilities|Struct|MakeGameplayEventPlayerExitedVehicle (Fortnite|GetFortPlayerControllerfromActor _asfort_player_pawn)))))))))))

(event Custom|OnVehicleAbilitiesRemoved
  (params (ExitedPlayerController :type Fort%20Player%20Controller%20Object%20Reference))
  ;[0 0.5151 1 0.5] Transmitters
  (bind _asfort_player_controller_athena (Utilities|Casting|CastToFortPlayerControllerAthena ExitedPlayerController)
    (:then
      (if (not (Class|FortPlayerControllerAthena|IsTeleportingtoLobby _asfort_player_controller_athena))
        (GameplayEvent|BroadcastEvent (Variables|UserOptions|Events|GetOn%20Player%20Exits%20Vehicle) 0 (Utilities|Struct|MakeGameplayEventPlayerExitedVehicle ExitedPlayerController)))
      (Vehicle|UnbindEventfromOnVehicleAbilitiesRemoved ExitedPlayerController (delegate Custom|OnVehicleAbilitiesRemoved)))
    (:CastFailed
      _)))

(event EventOnWorldReady
  ;[0.5541 1 0.6092 0.75] Handle Collision here based on visibility settings of the spawner
  (|CallParentFunction)
  (CallFunction|HandleEditModeCollision))

(event Custom|CheckEjectOnCooldown
  ;[1 0.5885 0.5885 0.8] Handle resetting the ejection cooldown if time has elapsed
  (Utilities|FlowControl|RetriggerableDelay (Variables|Default|GetEjectCooldownDelay))
  (Variables|Default|SetbEjectOnCooldown false))

(event Custom|CheckSeatingChannel
  ;[1 0.8379 0.211] When respawning on game start, set bool to attempt to seat the cahced pawn to seat if it exists after the vehicle respawns. The allows player spawn pads to seat a player on spawn even when the vehicle is in the process of respawning
  (Networking|SwitchHasAuthority
    (:Authority
      (Variables|Default|SetbPendingReseat true)
      (CallFunction|RefreshVehicleSpawn))))

(event Custom|CheckToSeatOnStart
  (bind _pendingdriver (Variables|Default|GetPendingDriver))
  ;[1 0.8379 0.211] When respawning on game start, set bool to attempt to seat the cahced pawn to seat if it exists after the vehicle respawns. The allows player spawn pads to seat a player on spawn even when the vehicle is in the process of respawning
  (Networking|SwitchHasAuthority
    (:Authority
      (if (Variables|Default|GetbPendingReseat)
        (Utilities|IsValid _pendingdriver
          (:Is%20Valid
            (CallFunction|SetDriveronStart _pendingdriver)
            (Variables|Default|SetPendingDriver 0))
          (:Is%20Not%20Valid
            _))
        (Variables|Default|SetbPendingReseat false)))))

(event Custom|SetDriverOnStart
  (params (NewDriver :type Fort%20Player%20Pawn%20Object%20Reference))
  ;[1 0.8379 0.211] When respawning on game start, set bool to attempt to seat the cahced pawn to seat if it exists after the vehicle respawns. The allows player spawn pads to seat a player on spawn even when the vehicle is in the process of respawning
  (Networking|SwitchHasAuthority
    (:Authority
      (Variables|Default|SetPawnToPlaceInVehicle NewDriver)
      (CallFunction|AttemptSeatPawn))))

(event Custom|EnabledOnReceived
  (params (Instigator :type Controller%20Object%20Reference))
  (bind _enabledcomponent (Variables|Default|GetEnabledComponent))
  ;[0.3018 0 0.5312] Receivers
  (Variables|Default|SetCachedInstigator Instigator)
  (if (not (Class|CreativeEnabledComponent|IsEnabled (Variables|Default|GetEnabledComponent)))
    (Utilities|StopOtherforDelay
      (:Out%20A
        (Class|CreativeEnabledComponent|SetEnabled _enabledcomponent true))
      (:Out%20B
        (Class|CreativeEnabledComponent|SetEnabled _enabledcomponent false))) ; When one input is triggered, pass through\nand stop the other input passing through\nfor VehicleMinToggleTime seconds
    (CallFunction|UpdateVehicleSpawnSettings true)))

(event Custom|DisableOnReceived
  (params (Instigator :type Controller%20Object%20Reference))
  (bind _enabledcomponent (Variables|Default|GetEnabledComponent))
  ;[0.3018 0 0.5312] Receivers
  (Variables|Default|SetCachedInstigator Instigator)
  (if (Class|CreativeEnabledComponent|IsEnabled (Variables|Default|GetEnabledComponent))
    (Utilities|StopOtherforDelay
      (:Out%20A
        (Class|CreativeEnabledComponent|SetEnabled _enabledcomponent true))
      (:Out%20B
        (Class|CreativeEnabledComponent|SetEnabled _enabledcomponent false))) ; When one input is triggered, pass through\nand stop the other input passing through\nfor VehicleMinToggleTime seconds
    (CallFunction|UpdateVehicleSpawnSettings true)))

(event Custom|RespawnOnReceived
  (params (Instigator :type Controller%20Object%20Reference))
  ;[0.3018 0 0.5312] Receivers
  (Variables|Default|SetCachedInstigator Instigator)
  (CallFunction|RespawnVehicleFromSettings))

(event Custom|DestroyOnReceived
  (params (Instigator :type Controller%20Object%20Reference))
  ;[0.3018 0 0.5312] Receivers
  (Utilities|IsValid (Utilities|ToObject(Interface) (Variables|Default|GetCachedVehicleInterface))
    (:Is%20Valid
      (Variables|Default|SetCachedInstigator Instigator)
      (Utilities|FlowControl|Delay 0.0)
      (CallFunction|ManuallyDestroyVehicle)
      (CallFunction|VehicleDestroyed 0))
    (:Is%20Not%20Valid
      (Networking|SwitchHasAuthority
        (:Authority
          (CallFunction|AttemptRespawnVehicle))))) ; This check is done in cast the cached vehicle is pending kill
)

(event Custom|AssignDriverOnReceived
  (params (Instigator :type Controller%20Object%20Reference))
  ;[0.3018 0 0.5312] Receivers
  (Utilities|IsValid Instigator
    (:Is%20Valid
      (bind _asfort_player_controller (Utilities|Casting|CastToFortPlayerController Instigator))
      (bind _asfort_player_pawn (Utilities|Casting|CastToFortPlayerPawn (Pawn|GetControlledPawn _asfort_player_controller)))
      (bind _returnvalue (Minigame|GetMinigameforActor _asfort_player_pawn))
      (Utilities|IsValid _returnvalue
        (:Is%20Valid
          (if (Minigame|IsPostGame _returnvalue)
            _
            (elif (Variables|Default|GetbPendingReseat)
              (Variables|Default|SetPendingDriver _asfort_player_pawn)
              (else
                (Variables|Default|SetPawnToPlaceInVehicle _asfort_player_pawn)
                (CallFunction|AttemptSeatPawn)))))))))

(event Custom|ApplyOffRoadTires
  (params (Instigator :type Controller%20Object%20Reference))
  (bind _tiremodtag (Variables|Default|GetTireModTag))
  (bind _cachedvehicleinterface (Variables|Default|GetCachedVehicleInterface))
  ;[0.3018 0 0.5312] Receivers
  (Utilities|IsValid (Utilities|ToObject(Interface) _cachedvehicleinterface)
    (:Is%20Valid
      (bind _asfort_athena_vehicle (Utilities|Casting|CastToFortAthenaVehicle _cachedvehicleinterface))
      (Utilities|VehicleHasTires _asfort_athena_vehicle
        (:HasTires
          (if (not (Vehicle|Mod|Tire|HasTireMod _asfort_athena_vehicle))
            (if (GameplayTags|IsGameplayTagValid _tiremodtag)
              (Vehicle|Mod|ApplyVehicleMod _asfort_athena_vehicle _tiremodtag 0) ; Used instead of ApplyTireMod for proper internal setup
            )))))))

(event Custom|Apply%20Road%20Tires
  (params (Instigator :type Controller%20Object%20Reference))
  (bind _cachedvehicleinterface (Variables|Default|GetCachedVehicleInterface))
  ;[0.3018 0 0.5312] Receivers
  (Utilities|IsValid (Utilities|ToObject(Interface) _cachedvehicleinterface)
    (:Is%20Valid
      (bind _asfort_athena_vehicle (Utilities|Casting|CastToFortAthenaVehicle _cachedvehicleinterface))
      (Utilities|VehicleHasTires _asfort_athena_vehicle
        (:HasTires
          (if (GameplayTags|IsGameplayTagValid (Variables|Default|GetTireModTag))
            (if (Vehicle|Mod|Tire|HasTireMod _asfort_athena_vehicle)
              (Vehicle|Mod|Tire|RemoveAppliedTireMod _asfort_athena_vehicle))))))))

(event Custom|RepairAllTires
  (params (Instigator :type Controller%20Object%20Reference))
  (bind _cachedvehicleinterface (Variables|Default|GetCachedVehicleInterface))
  ;[0.3018 0 0.5312] Receivers
  (Utilities|IsValid (Utilities|ToObject(Interface) _cachedvehicleinterface)
    (:Is%20Valid
      (bind _asfort_athena_vehicle (Utilities|Casting|CastToFortAthenaVehicle _cachedvehicleinterface))
      (Utilities|VehicleHasTires _asfort_athena_vehicle
        (:HasTires
          (for _array_element (Variables|/Script/FortniteGame.FortVehicleConfigs|GetDamageableParts (Class|FortVehicleInterface|GetVehicleConfigs _cachedvehicleinterface))
            (bind (_array_element_shapename _array_element_bonename _array_element_parttag) (Utilities|Struct|BreakVehicleDamageablePartConfig _array_element))
            (if (GameplayTags|MatchesTag _array_element_parttag "(TagName=\"Vehicle.Part.Tire\")")
              (Class|FortAthenaVehicle|RepairDamageablePart _asfort_athena_vehicle _array_element_shapename 1.0) ; Doesn't function correctly yet.  See FORT-600140
            )))))))

(event Custom|PopAllTires
  (params (Instigator :type Controller%20Object%20Reference))
  (bind _cachedvehicleinterface (Variables|Default|GetCachedVehicleInterface))
  ;[0.3018 0 0.5312] Receivers
  (Utilities|IsValid (Utilities|ToObject(Interface) _cachedvehicleinterface)
    (:Is%20Valid
      (bind _asfort_athena_vehicle (Utilities|Casting|CastToFortAthenaVehicle _cachedvehicleinterface))
      (Utilities|VehicleHasTires _asfort_athena_vehicle
        (:HasTires
          (for _array_element (Variables|/Script/FortniteGame.FortVehicleConfigs|GetDamageableParts (Class|FortVehicleInterface|GetVehicleConfigs _cachedvehicleinterface))
            (bind (_array_element_shapename _array_element_bonename _array_element_parttag) (Utilities|Struct|BreakVehicleDamageablePartConfig _array_element))
            (if (GameplayTags|MatchesTag _array_element_parttag "(TagName=\"Vehicle.Part.Tire\")")
              (Class|FortAthenaVehicle|BreakDamageablePart _asfort_athena_vehicle _array_element_shapename))))))))

(event Custom|FullyRepairVehicle
  (params (Instigator :type Controller%20Object%20Reference))
  (bind _cachedvehicleinterface (Variables|Default|GetCachedVehicleInterface))
  ;[0.3018 0 0.5312] Receivers
  (Utilities|IsValid (Utilities|ToObject(Interface) _cachedvehicleinterface)
    (:Is%20Valid
      (bind _asfort_athena_vehicle (Utilities|Casting|CastToFortAthenaVehicle _cachedvehicleinterface))
      (Class|FortAthenaVehicle|RepairAllDamageableParts _asfort_athena_vehicle 1.0)
      (Class|FortVehicleInterface|SetHealthPercent _asfort_athena_vehicle 1.0))))

(event EventOnActorMovedToNewLocation
  ;[1 1 1] Respawn vehicle if device changed it's location during edit mode
  (if (not (Minigame|IsMinigameinProgress (Variables|Default|GetFortMinigameProgress)))
    (CallFunction|RespawnVehicleFromSettings)))

(event Custom|Load%20Curve%20Table
  (params (AssetToLoad :type Soft%20Object%20Reference))
  (bind _object (Utilities|AsyncLoadAsset AssetToLoad)
    (:Completed
      (CallFunction|UpdateDataRegistryFromCurveTable _object))))

(fn AttemptSeatPawn
  (bind _localathenavehicle (LocalVariables|GetLocalAthenaVehicle))
  (bind _returnvalue (Vehicle|GetDriver _localathenavehicle))
  (bind _pawntoplaceinvehicle (Variables|Default|GetPawnToPlaceInVehicle))
  (bind _localathenavehicle_2 (LocalVariables|GetLocalAthenaVehicle))
  ;[1 0.8379 0.211] Handle seating a pawn based on if there is a driver, and if there is what team the driver is on
  (Utilities|IsValid (Variables|Default|GetPawnToPlaceInVehicle)
    (:Is%20Valid
      (bind _asfort_athena_vehicle (Utilities|Casting|CastToFortAthenaVehicle (Variables|Default|GetCachedVehicleInterface)))
      (Utilities|IsValid _asfort_athena_vehicle
        (:Is%20Valid
          (if (or (Class|FortAthenaVehicle|IsPendingDeath _asfort_athena_vehicle) (<= (Class|FortVehicleInterface|GetHealth _asfort_athena_vehicle) 0.0))
            (Development|PrintString "Pending Death - Don't seat new pawn")
            (else
              (CallFunction|CancelAbilitieswithTags)
              (LocalVariables|SetLocalAthenaVehicle _asfort_athena_vehicle)
              (Utilities|FlowControl|Sequence
                (:then_0
                  (Utilities|IsValid _returnvalue
                    (:Is%20Valid
                      (if (Team|OnSameTeam _returnvalue (Variables|Default|GetPawnToPlaceInVehicle))
                        (bind _returnvalue_2 (Vehicle|ForceSeatSinglePawn _localathenavehicle_2 _pawntoplaceinvehicle -1))
                        (LocalVariables|SetbPawnSeated _returnvalue_2)
                        (else
                          (CallFunction|CheckEjectonCooldown)
                          (if (not (Variables|Default|GetbEjectOnCooldown))
                            (Vehicle|EjectSeat _localathenavehicle_2 0) ; Eject the current driver if they are on another team
                            (bind _returnvalue_3 (Vehicle|ForceSeatSinglePawn _localathenavehicle_2 _pawntoplaceinvehicle -1))
                            (LocalVariables|SetbPawnSeated _returnvalue_3)
                            (Variables|Default|SetbEjectOnCooldown true)))))
                    (:Is%20Not%20Valid
                      (if (Utilities|Array|ContainsItem (Vehicle|GetAllPassengers (LocalVariables|GetLocalAthenaVehicle)) (Variables|Default|GetPawnToPlaceInVehicle))
                        (Vehicle|RequestChangeSeatonServer (LocalVariables|GetLocalAthenaVehicle) (Pawn|GetController (Variables|Default|GetPawnToPlaceInVehicle)) 0)
                        (LocalVariables|SetbPawnSeated false) ; Avoid triggering the delegate since we were already at the vehicle
                        (else
                          (bind _returnvalue_4 (Vehicle|ForceSeatSinglePawn (LocalVariables|GetLocalAthenaVehicle) (Variables|Default|GetPawnToPlaceInVehicle) -1))
                          (LocalVariables|SetbPawnSeated _returnvalue_4))))) ; Is there currently a driver
                )
                (:then_1
                  (if (LocalVariables|GetbPawnSeated)
                    (CallFunction|OnPawnEnteredVehicle (Variables|Default|GetCachedVehicleInterface) (Variables|Default|GetPawnToPlaceInVehicle)) ; Our method of force seating the player does not\ntrigger the delegate we listen to, so we manually call it here
                    (else
                      _))
                  (Variables|Default|SetPawnToPlaceInVehicle 0) ; null global variable now that it has been used
                )) ; Null pawn to seat after attempting to seat
            )))) ; Cast can pass even if vehicle is destroyed \nso this will make sure the vehicle still exists
    )
    (:Is%20Not%20Valid
      (Development|PrintString "Failed to place pawn in vehicle. Pawn Invalid" true true "(R=1.000000,G=0.039115,B=0.000000,A=1.000000)"))))

(fn BuildDataRegistryResolverScope
  (return (Variables|Default|GetResolver%20Scopes) 1 true))

(fn CancelAbilitiesWithTags
  (bind _pawntoplaceinvehicle (Variables|Default|GetPawnToPlaceInVehicle))
  (bind _returnvalue (Ability|GetAbilitySystemComponent _pawntoplaceinvehicle))
  (bind (_asfort_ability_system_component _bsuccess) (Utilities|Casting|CastToFortAbilitySystemComponent _returnvalue))
  (if _bsuccess
    (Tags|CancelAbilitiesWithTags _asfort_ability_system_component (LocalVariables|GetPossessedPropTag))))

(fn DestroySpawnedVehicle
  (bind _cachedvehicleinterface (Variables|Default|GetCachedVehicleInterface))
  (bind _returnvalue (Utilities|ToObject(Interface) _cachedvehicleinterface))
  ;[1 0.1543 0] Destroys the current vehicle if it exists
  (Utilities|IsValid _returnvalue
    (:Is%20Valid
      (bind _asfort_athena_vehicle (Utilities|Casting|CastToFortAthenaVehicle _returnvalue))
      (Vehicle|ForceDestroyVehicle _asfort_athena_vehicle false))) ; This check is done in cast the cached vehicle is pending kill
)

(fn GetEnabledStateasIndex
  (return (DeviceConversion|EnabledPhaseEnumtoLegacyInt (Variables|UserOptions|GetEnabled%20During%20Phase))))

(fn GetSupportsWraps
  (return false))

(fn GetBoostEnabled
  (return false))

(fn GetBoostFuelUseMultiplier
  (return 1.0))

(fn GetCacheContextOverride
  (return 0))

(fn GetCollisionStaticMesh
  (bind _staticmeshcomponent (Variables|Debug|GetStaticMeshComponent))
  (if (Utilities|IsValid _staticmeshcomponent :!class /Script/Engine.KismetSystemLibrary)
    (return "0, 0, 0" (Variables|/Script/Engine.StaticMeshComponent|GetStaticMesh _staticmeshcomponent))
    (else
      (return "0, 0, 0" 0))))

(fn GetConsumesFuel
  (return (not (Variables|Legacy|GetbHasInfiniteFuel))))

(fn GetCreativeActorBounds
  (bind _boundextentsoverride (Variables|Default|GetBoundExtentsOverride))
  (bind _boundstransform_matrix (Variables|GetBoundsTransform_Matrix))
  (bind (_xplane _yplane _zplane) (Utilities|Struct|BreakMatrix _boundstransform_matrix))
  (bind (_w _x _y _z) (Utilities|Struct|BreakPlane _zplane))
  (bind (_w_2 _x_2 _y_2 _z_2) (Utilities|Struct|BreakPlane _xplane))
  (bind (_w_3 _x_3 _y_3 _z_3) (Utilities|Struct|BreakPlane _yplane))
  (bind (_boundsorigin_x _boundsorigin_y _boundsorigin_z) (Math|Vector|BreakVector (Variables|GetBoundsOrigin_X)))
  (bind (_boundsextent_x _boundsextent_y _boundsextent_z) (Math|Vector|BreakVector (Variables|GetBoundsExtent_X)))
  ; GetInitialBoundsValues
  (Variables|SetBoundsMin (Math|Vector|NegateVector _boundextentsoverride))
  (Variables|SetBoundsMax _boundextentsoverride)
  (Variables|SetBoundsOrigin (* (+ (Variables|GetBoundsMax) (Variables|GetBoundsMin)) 0.5))
  (Variables|SetBoundsExtent (* (- (Variables|GetBoundsMax) (Variables|GetBoundsMin)) 0.5))
  ; GetTransformForBounds
  (Variables|SetBoundsTransform_Matrix (Math|Transform|ToMatrix(Transform) (Math|Transform|MakeTransform "0, 0, 0" (Transformation|GetActorRotation self) (Transformation|GetActorScale3D self))))
  (Variables|SetMatrix_XVector (Math|Vector|MakeVector _x_2 _y_2 _z_2)) ; Transform's X component
  (Variables|SetMatrix_YVector (Math|Vector|MakeVector _x_3 _y_3 _z_3)) ; Transform's Y component
  (Variables|SetMatrix_ZVector (Math|Vector|MakeVector _x _y _z)) ; Transform's Z component
  ; GetNewBoundsOrigin
  (bind _output_get (Variables|SetBoundsOrigin_New (Utilities|Operators|Add (* (Variables|GetMatrix_XVector) _boundsorigin_x) (* (Variables|GetMatrix_YVector) _boundsorigin_y) (* (Variables|GetMatrix_ZVector) _boundsorigin_z))))
  ; GetNewBoundsExtent
  (bind _output_get_2 (Variables|SetBoundsExtent_New (Math|Vector|VectorGetAbs (* (Variables|GetMatrix_XVector) _boundsextent_x))))
  (bind _output_get_3 (Variables|SetBoundsExtent_New (+ _output_get_2 (Math|Vector|VectorGetAbs (* (Variables|GetMatrix_YVector) _boundsextent_y)))))
  (bind _output_get_4 (Variables|SetBoundsExtent_New (+ _output_get_3 (Math|Vector|VectorGetAbs (* (Variables|GetMatrix_ZVector) _boundsextent_z)))))
  (return (Variables|Default|GetbOverrideBoundExtents) (+ _output_get_4 _output_get)))

(fn GetCreativeActorOrigin
  (return (Variables|Default|GetbOverrideBoundOrigin) (Variables|Default|GetBoundOriginOverride)))

(fn GetCreativeActorRelativeTransform
  (return false 0))

(fn GetDriver
  (bind _asfort_athena_vehicle (Utilities|Casting|CastToFortAthenaVehicle (Variables|Default|GetCachedVehicleInterface))
    (:then
      (Utilities|IsValid (Vehicle|GetDriver _asfort_athena_vehicle)
        (:Is%20Valid
          (Utilities|IsValid (Pawn|GetController (Vehicle|GetDriver _asfort_athena_vehicle))
            (:Is%20Valid
              (return (Pawn|GetController (Vehicle|GetDriver _asfort_athena_vehicle))))
            (:Is%20Not%20Valid
              _)))
        (:Is%20Not%20Valid
          _))
      (return (LocalVariables|GetNullController)))
    (:CastFailed
      _)))

(fn GetFuelCapacity
  (bind _asfort_athena_vehicle (Utilities|Casting|CastToFortAthenaVehicle (Variables|Default|GetCachedVehicleInterface)))
  (return (Vehicle|Fuel|GetFuelCapacity _asfort_athena_vehicle)))

(fn GetFuelUseMultiplier
  (return 1.0))

(fn GetFuelValue
  (bind _asfort_athena_vehicle (Utilities|Casting|CastToFortAthenaVehicle (Variables|Default|GetCachedVehicleInterface)))
  (return (Vehicle|Fuel|GetFuel _asfort_athena_vehicle)))

(fn GetGridAxisGuideAttachmentPoint
  (return false false "0, 0, 0"))

(fn GetMaxRandomStartingFuel
  (return 0))

(fn GetMeshComponents
  (return 0))

(fn GetMinRandomStartingFuel
  (return 0))

(fn GetPassengers
  ;[1 0 0] Not Exposed in Verse ATM due to not being able to return arrays from BP functions
  (bind _asfort_athena_vehicle (Utilities|Casting|CastToFortAthenaVehicle (Variables|Default|GetCachedVehicleInterface)))
  (for _array_element (Vehicle|GetAllNonDriverPassengers _asfort_athena_vehicle :!class /Script/FortniteGame.FortAthenaVehicle)
    (Utilities|Array|Add (LocalVariables|GetPassengerControllers) (Pawn|GetController _array_element)))
  (return (LocalVariables|GetPassengerControllers)))

(fn GetRandomStartingFuel
  (return false))

(fn GetSpawnedVehicle
  (bind _asfort_athena_vehicle (Utilities|Casting|CastToFortAthenaVehicle (Variables|Default|GetCachedVehicleInterface))
    (:then
      (return _asfort_athena_vehicle))
    (:CastFailed
      (return (LocalVariables|GetNull%20Vehicle)))))

(fn GetStartingFuelValue
  (return 100))

(fn GetVehicleHealth
  (bind _asfort_athena_vehicle (Utilities|Casting|CastToFortAthenaVehicle (Variables|Default|GetCachedVehicleInterface)))
  (return (Class|FortVehicleInterface|GetHealth _asfort_athena_vehicle)))

(fn GetVisibilityComponents
  (return (Utilities|Array|MakeArray (Variables|Debug|GetStaticMeshComponent))))

(fn GetVisibilityComponentsWithCollisionOverrides
  (return (Utilities|Map|MakeMap (Variables|Debug|GetStaticMeshComponent) "QueryAndPhysics")))

(fn HandleEditorVisibility
  ;[0.8827 0.1953 1 0.25] We only want to truly hide the base if we're in play mode. In the editor and edit mode the base should still appear as a hologram.
  (if (Editor|InEditor)
    (Rendering|SetVisibility (Variables|Debug|GetStaticMeshComponent) true false)
    (else
      (CallFunction|RefreshVisibility))))

(fn HandleEditModeCollision
  (bind _returnvalue (Building|IsPreviewBuildingActor))
  (bind _visible (ChildFunctions|IsVisibleInGame))
  (bind _returnvalue_2 (or _returnvalue _visible))
  (bind _returnvalue_3 (not _returnvalue_2))
  (if (not (CallFunction|IsinPlayMode))
    (bind _visibilitycomponents (CallFunction|GetVisibilityComponents))
    (for _array_element _visibilitycomponents
      (Utilities|IsValid _array_element
        (:Is%20Valid
          (bind _asstatic_mesh_component (Utilities|Casting|CastToStaticMeshComponent _array_element))
          (Utilities|FlowControl|Sequence
            (:then_0
              (if _returnvalue_2
                (Collision|SetCollisionProfileName _asstatic_mesh_component "FortBuildingMeshPhysics" true :!class /Script/Engine.PrimitiveComponent)
                (Rendering|Material|SetMaterial _asstatic_mesh_component 0 "/Game/Creative/Devices/VehicleSpawner/MI_Floor_Spawnpad_Creative_Vehicle_Inst.MI_Floor_Spawnpad_Creative_Vehicle_Inst")
                (Rendering|Material|SetMaterial _asstatic_mesh_component 1 "/Engine/EngineMaterials/WorldGridMaterial.WorldGridMaterial")
                (else
                  (Collision|SetCollisionProfileName _asstatic_mesh_component "FortTriggerOnlyInteractions" true :!class /Script/Engine.PrimitiveComponent)
                  (Rendering|Material|SetMaterial _asstatic_mesh_component 0 "/Game/Creative/Devices/Generic/Materials/M_CP_Device_Hologram_Vehicles.M_CP_Device_Hologram_Vehicles")
                  (Rendering|Material|SetMaterial _asstatic_mesh_component 1 "/Game/Creative/Devices/Generic/Materials/M_CP_Device_Hologram_Vehicles.M_CP_Device_Hologram_Vehicles")
                  (Collision|SetCollisionResponsetoChannel _asstatic_mesh_component "ECC_Visibility" "ECR_Overlap"))))
            (:then_1
              ;[1 1 1] Setting here because setting the mesh collision alone is causing issues where the client and server don't agree
              (Variables|Mesh|SetbNoPhysicsCollision _returnvalue_3)
              (Variables|Mesh|SetbNoCameraCollision _returnvalue_3)
              (Variables|Mesh|SetbNoPawnCollision _returnvalue_3)
              (Variables|Mesh|SetbNoWeaponCollision _returnvalue_3))))))))

(fn HandleVehicleReferenceVisibility
  (params (IsForPreviewing :type Boolean))
  (bind _previewvehiclemesh (Variables|Default|GetPreviewVehicleMesh))
  ;[0.6214 1 0.9501] Set visibility of the reference (placement only) for the spawner according to MoveTool usage
  (if IsForPreviewing
    ;[0.6436 0 1] Only spawn BP_VehiclePreview If can execute cosmetic events is true
    (Networking|CanExecuteCosmeticEvents
      (:True
        (Collision|SetCollisionEnabled (Variables|Default|GetBox) "QueryOnly")
        (if (Utilities|IsValidSoftObjectReference _previewvehiclemesh)
          ;[0.6156 0.291 1] Spawn BP_VehicleSpawnPreview at location of Vehicle Reference
          (bind _returnvalue (Game|SpawnActorfromClass "/CRD_VehicleSpawners/BP_VehicleSpawnerPreview.BP_VehicleSpawnerPreview_C" (Transformation|GetWorldTransform (Variables|Default|GetSpawnLocation)) "Undefined" "OverrideRootScale" self (Variables|Default|GetBox) _previewvehiclemesh 0))
          (Transformation|AttachActorToActor _returnvalue self "None" "KeepWorld" "KeepWorld" "KeepWorld" true))))))

(fn HasAdditionalValidPlacementLogic
  (return true))

(fn HasVehicle
  (return (Utilities|IsValid (Utilities|ToObject(Interface) (Variables|Default|GetCachedVehicleInterface)) :!class /Script/Engine.KismetSystemLibrary)))

(fn IgnoreWorldGridScaleOverride
  (return false))

(fn InitializeDataRegistryOptions
  (Class|FortAthenaMutatorUtilities|RemoveDataRegistryOptionsforContextObject)
  (Class|FortAthenaMutatorUtilities|ApplyActorOptionsComponenttoDataRegistry (Variables|Default|GetToyOptionsComponent)))

(fn InitializeDataRegistryResolverScope
  (Networking|SwitchHasAuthority
    (:Authority
      (Utilities|Array|Clear (Variables|Default|GetResolver%20Scopes))
      (Variables|Default|SetResolver%20Scopes (Utilities|Array|MakeArray (SpatialGameplay|GetSpatialGameplayInterfaceFNameforActor self) (Utilities|String|StringToName (Utilities|GetClassDisplayName (Utilities|GetClass self))) (Utilities|String|StringToName (Asset|GetSoftPtrAssetName (Variables|Default|GetVehicle%20Curve%20Table%20Reference))))))))

(fn InitializeVehicleHealth
  (params (In%20Vehicle :type Fort%20Athena%20Vehicle%20Object%20Reference))
  (bind _in_vehicle (Variables|GetIn%20Vehicle))
  (bind _returnvalue (Ability|GetFortAbilitySystemComponent _in_vehicle))
  (bind _vehicle_indestructible (Variables|UserOptions|GetVehicle%20Indestructible))
  ;[0.7716 1 0.6029] Override Health as set by User Options.  Don't worry about clearing the Effect, as changing the options results in the vehicle respawning anyway
  (Utilities|IsValid In%20Vehicle
    (:Is%20Valid
      (if (!= (Variables|UserOptions|GetVehicle%20Health) 0.0)
        (Fortnite|SetCanbeDamaged (Variables|GetIn%20Vehicle) (not _vehicle_indestructible))
        (LocalVariables|SetHealth (Math|Float|SelectFloat (Variables|UserOptions|GetVehicle%20Health) 1e+06 (not _vehicle_indestructible)))
        (Networking|SwitchHasAuthority
          (:Authority
            (Utilities|IsValid _returnvalue
              (:Is%20Valid
                (bind _returnvalue_2 (Ability|GameplayEffect|AssignTagSetbyCallerMagnitude (GameplayEffects|MakeOutgoingSpec :self _returnvalue :GameplayEffectClass "/Game/Creative/Devices/Common/GameplayEffects/GE_Devices_Common_SetMaxHealth.GE_Devices_Common_SetMaxHealth_C" :Level 0.0) "(TagName=\"SetByCaller.MaxHealth\")" (LocalVariables|GetHealth)))
                (GameplayEffects|ApplyGameplayEffectSpecToSelf _returnvalue _returnvalue_2)))))))))

(fn InitializeVehicleVisual
  (params (InVehicle :type Fort%20Athena%20Vehicle%20Object%20Reference))
  ;[1 1 1] Override vehicle visuals and overridden on child functions
  ; Start with built-in variants like SportBike & DirtBike
  ; Then extended to custom material / meshes when we figured out the proper path
  (return))

(fn IsVisibleInGame
  (return (Variables|UserOptions|Basic|GetVisible%20During%20Game)))

(fn IsDriver
  (params (Agent :type Controller%20Object%20Reference))
  (bind _cachedvehicleinterface (Variables|Default|GetCachedVehicleInterface))
  (Utilities|IsValid Agent
    (:Is%20Valid
      (Utilities|IsValid (Utilities|ToObject(Interface) _cachedvehicleinterface)
        (:Is%20Valid
          (return (== (Pawn|GetControlledPawn Agent) (Vehicle|GetDriver _cachedvehicleinterface))))
        (:Is%20Not%20Valid
          _)))
    (:Is%20Not%20Valid
      _))
  (return false))

(fn IsPassenger
  (params (Agent :type Controller%20Object%20Reference))
  (bind _returnvalue (Pawn|GetControlledPawn Agent))
  (bind _cachedvehicleinterface (Variables|Default|GetCachedVehicleInterface))
  (Utilities|IsValid Agent
    (:Is%20Valid
      (Utilities|IsValid (Utilities|ToObject(Interface) _cachedvehicleinterface)
        (:Is%20Valid
          (for _array_element (Vehicle|GetAllNonDriverPassengers _cachedvehicleinterface :!class /Script/FortniteGame.FortVehicleInterface)
            (if (== _array_element _returnvalue)
              (return (== _array_element _returnvalue)))))
        (:Is%20Not%20Valid
          _))
      (return false))
    (:Is%20Not%20Valid
      _)))

(fn IsValidPlacement
  (return true))

(fn LoadCurveTableifAble
  (bind _vehicle_curve_table_reference (Variables|Default|GetVehicle%20Curve%20Table%20Reference))
  (if (Utilities|IsValidSoftObjectReference _vehicle_curve_table_reference)
    (CallFunction|LoadCurveTable _vehicle_curve_table_reference)))

(fn MigrateLegacyUserOptions
  (bind _startingfuel (Class|GetClassDefaults "/CRD_VehicleSpawners/Device_VehicleSpawnerParent_V2.Device_VehicleSpawnerParent_V2_C"))
  (bind _spawndelay (Class|GetClassDefaults "/CRD_VehicleSpawners/Device_VehicleSpawnerParent_V2.Device_VehicleSpawnerParent_V2_C"))
  (bind _maxvehiclehealth (Class|GetClassDefaults "/CRD_VehicleSpawners/Device_VehicleSpawnerParent_V2.Device_VehicleSpawnerParent_V2_C"))
  (bind (_enablevehiclecosmetics _data_registry_name _resolver_scopes _physics_vehicle_config_class _vehicle_curve_table_reference _bisvisible) (Class|GetClassDefaults "/CRD_VehicleSpawners/Device_VehicleSpawnerParent_V2.Device_VehicleSpawnerParent_V2_C"))
  (bind _bvehiclesupportswraps (Class|GetClassDefaults "/CRD_VehicleSpawners/Device_VehicleSpawnerParent_V2.Device_VehicleSpawnerParent_V2_C"))
  (bind _boostfuelusemultiplier (Class|GetClassDefaults "/CRD_VehicleSpawners/Device_VehicleSpawnerParent_V2.Device_VehicleSpawnerParent_V2_C"))
  (bind _bhasinfinitefuel (Class|GetClassDefaults "/CRD_VehicleSpawners/Device_VehicleSpawnerParent_V2.Device_VehicleSpawnerParent_V2_C"))
  (bind _coloroption (Class|GetClassDefaults "/CRD_VehicleSpawners/Device_VehicleSpawnerParent_V2.Device_VehicleSpawnerParent_V2_C"))
  (bind _waterdestructiondelay (Class|GetClassDefaults "/CRD_VehicleSpawners/Device_VehicleSpawnerParent_V2.Device_VehicleSpawnerParent_V2_C"))
  ;[1 1 1] Not Finished
  (|CallParentFunction)
  (Utilities|FlowControl|Sequence
    (:then_0
      (Utilities|FlowControl|Sequence
        (:then_0
          (if (!= (Variables|Legacy|GetSpawnDelay) _spawndelay)
            (if (== (Variables|Legacy|GetSpawnDelay) -1.0)
              (Variables|UserOptions|SetEnableRespawn false)
              (else
                (Variables|UserOptions|SetRespawn%20Time (Variables|Legacy|GetSpawnDelay))))
            (Variables|Legacy|SetSpawnDelay _spawndelay)
            (LocalVariables|SetOptionMigrated true)))
        (:then_1
          (if (!= (Variables|Legacy|GetMaxVehicleHealth) _maxvehiclehealth)
            (if (== (Variables|Legacy|GetMaxVehicleHealth) -1.0)
              (Variables|UserOptions|SetVehicle%20Indestructible true)
              (else
                (Variables|UserOptions|SetVehicle%20Health (Variables|Legacy|GetMaxVehicleHealth))))
            (Variables|Legacy|SetMaxVehicleHealth _maxvehiclehealth)
            (LocalVariables|SetOptionMigrated true)))
        (:then_2
          (if (!= (Variables|Legacy|GetWaterDestructionDelay) _waterdestructiondelay)
            (if (== (Variables|Legacy|GetWaterDestructionDelay) -1.0)
              (Variables|UserOptions|SetDestroy%20when%20Stuck%20Underwater false)
              (Variables|UserOptions|SetWater%20Destruction%20Timer _waterdestructiondelay)
              (else
                (Variables|UserOptions|SetWater%20Destruction%20Timer (Variables|Legacy|GetWaterDestructionDelay))))
            (Variables|Legacy|SetWaterDestructionDelay _waterdestructiondelay)
            (LocalVariables|SetOptionMigrated true)))
        (:then_3
          (if (!= (Variables|Legacy|GetStartingFuel) _startingfuel)
            (ChildFunctions|SetRandomStartingFuel false)
            (ChildFunctions|SetStartingFuelValue (Variables|Legacy|GetStartingFuel))
            (Variables|Legacy|SetStartingFuel _startingfuel)
            (LocalVariables|SetOptionMigrated true)))
        (:then_4
          (if (!= (Variables|Legacy|GetBoostFuelUseMultiplier) _boostfuelusemultiplier)
            (if (== (Variables|Legacy|GetBoostFuelUseMultiplier) -1.0)
              (ChildFunctions|SetBoostEnabled false)
              (else
                (ChildFunctions|SetBoostEnabled true)
                (ChildFunctions|SetBoostFuelUseMultiplier (Variables|Legacy|GetBoostFuelUseMultiplier))))
            (Variables|Legacy|SetBoostFuelUseMultiplier _boostfuelusemultiplier)
            (LocalVariables|SetOptionMigrated true)))
        (:then_5
          (if (!= (Variables|Legacy|GetbVehicleSupportsWraps) _bvehiclesupportswraps)
            (ChildFunctions|SetSupportsWraps false)
            (Variables|Legacy|SetbVehicleSupportsWraps _bvehiclesupportswraps)
            (LocalVariables|SetOptionMigrated true)))
        (:then_6
          (if (!= (Variables|Legacy|GetColorOption) _coloroption)
            (ChildFunctions|SetColorOption (Variables|Legacy|GetColorOption))
            (Variables|Legacy|SetColorOption _coloroption)
            (LocalVariables|SetOptionMigrated true)))
        (:then_7
          (if (!= (Variables|Legacy|GetbHasInfiniteFuel) _bhasinfinitefuel)
            (ChildFunctions|SetConsumeFuel true)
            (Variables|Legacy|SetbHasInfiniteFuel _bhasinfinitefuel)
            (LocalVariables|SetOptionMigrated true)))
        (:then_8
          (bind _migratedproperties (DeviceConversion|MigrateLegacyEnabledIndex (Variables|UserOptions|GetEnabled%20During%20Phase) (Class|GetClassDefaults "/CRD_VehicleSpawners/Device_VehicleSpawnerParent_V2.Device_VehicleSpawnerParent_V2_C")))
          (if _migratedproperties
            (LocalVariables|SetOptionMigrated false)))
        (:then_9
          (if (!= _bisvisible (Variables|LegacyOptions|GetbIsVisible))
            (Variables|UserOptions|Basic|SetVisible%20During%20Game false)
            (Variables|LegacyOptions|SetbIsVisible _bisvisible)
            (LocalVariables|SetOptionMigrated true)))))
    (:then_1
      (if (LocalVariables|GetOptionMigrated)
        (OptionData|RefreshPropertyOverrides (Variables|Default|GetToyOptionsComponent))))))

(fn OnCooldown
  ;[0.9918 1 0.2351] Determine if the cooldown should be active. Works on both server and client by using the replicated Last Cooldown Start value
  (if (Class|CreativeCommonDeviceFunctionLibrary|IsLogicinPlayMode (Variables|Default|GetFortMinigameProgress))
    (return (and (Class|CreativeEnabledComponent|IsEnabled (Variables|Default|GetEnabledComponent)) (or (Math|Boolean|ANDBoolean (> (+ (Variables|Default|GetLastCooldownStart) (Variables|UserOptions|GetRespawn%20Time)) (Fortnite|Sync|Time|GetServerWorldTimeSeconds)) (> (Variables|UserOptions|GetRespawn%20Time) 0.0) (> (Variables|Default|GetLastCooldownStart) 0.0)) (not (Variables|UserOptions|GetEnableRespawn)))))
    (else
      (return false))))

(fn OnRep_LastCooldownStart
  ;[0.3417 1 0.9502] Start/Stop the cooldown flash effects
  (CallFunction|UpdateEffects))

(fn RefreshVehicleSpawn
  ;[0.2081 1 0.9801] Destroys the vehicle if it exists, or spawns one with no cooldown if there is no vehicle. Used when transitioning between edit and play mode
  (Networking|SwitchHasAuthority
    (:Authority
      (Utilities|Time|ClearandInvalidateTimerbyHandle (Variables|Default|GetVehicle%20Spawn%20Timer))
      (Variables|Default|SetInstant%20Respawn%20Override true)
      (Utilities|IsValid (Utilities|ToObject(Interface) (Variables|Default|GetCachedVehicleInterface))
        (:Is%20Valid
          (bind _asfort_athena_vehicle (Utilities|Casting|CastToFortAthenaVehicle (Variables|Default|GetCachedVehicleInterface)))
          (Vehicle|UnbindEventfromOnPawnExitedVehicleSeat _asfort_athena_vehicle (EventDispatchers|CreateEvent :!event On%20Pawn%20Exited%20Vehicle)) ; Make sure we don't transmit when exiting vehicle between states
          (CallFunction|DestroySpawnedVehicle))
        (:Is%20Not%20Valid
          (CallFunction|RespawnVehicleFromSettings))))))

(fn RefreshVisibility
  (bind _fortminigameprogress (Variables|Default|GetFortMinigameProgress))
  (bind _play_mode (Class|CreativeCommonDeviceFunctionLibrary|IsLogicinPlayMode _fortminigameprogress))
  (bind _returnvalue (not _play_mode))
  (bind _visible_during_game (Variables|UserOptions|Basic|GetVisible%20During%20Game))
  (bind _returnvalue_2 (or _returnvalue _visible_during_game))
  (bind _staticmeshcomponent (Variables|Debug|GetStaticMeshComponent))
  (Rendering|SetVisibility _staticmeshcomponent _returnvalue_2 false)
  (Development|SetHiddeninGame _staticmeshcomponent (not _returnvalue_2) false))

(fn RespawnVehiclefromSettings
  ;[0.25 0.3348 1] Destroys the current vehicle and prepares a new one to be spawned if the device is enabled or Instant Respawn Override is set
  (Networking|SwitchHasAuthority
    (:Authority
      (if (Variables|Default|GetInstant%20Respawn%20Override)
        (Variables|Default|SetInstant%20Respawn%20Override false)
        (else
          _))
      (if (Class|CreativeEnabledComponent|IsEnabled (Variables|Default|GetEnabledComponent))
        (CallFunction|ManuallyDestroyVehicle)
        (bind _asfort_athena_vehicle (Utilities|Casting|CastToFortAthenaVehicle (Variables|Default|GetCachedVehicleInterface))
          (:then
            (Utilities|IsValid _asfort_athena_vehicle
              (:Is%20Valid
                (Game|BindEventtoOnEndPlay _asfort_athena_vehicle (EventDispatchers|CreateEvent :!event Throttle%20New%20Vehicle%20On%20Vehicle%20End%20Play)))
              (:Is%20Not%20Valid
                _)))
          (:CastFailed
            _))
        (CallFunction|ThrottleNewVehicle)))))

(fn SetBoostFuelUseMultiplier
  (params (In%20Value :type Float%20%28double-precision%29))
  (Variables|ChildFunction|SetBoostFuelUse In%20Value))

(fn SetConsumeFuel
  (params (ConsumeFuel :type Boolean)))

(fn SetPhysicsVehicleConfigifAble
  (params (InVehicle :type Fort%20Athena%20Vehicle%20Object%20Reference))
  (bind _physics_vehicle_config_class (Variables|Default|GetPhysics%20Vehicle%20Config%20Class))
  (if (Utilities|IsValidClass _physics_vehicle_config_class)
    (Class|FortAthenaVehicle|SetPhysicsVehicleConfigClass (Variables|GetInVehicle) _physics_vehicle_config_class)))

(fn SetRandomStartingFuel
  (params (RandomStartingFuel :type Boolean))
  (Variables|ChildFunction|SetRandomStartingFuel RandomStartingFuel))

(fn SetStartingFuelValue
  (params (Starting%20Fuel%20Value :type Integer))
  (Variables|ChildFunction|SetStarting%20Fuel%20Value Starting%20Fuel%20Value))

(fn SetSupportsWraps
  (params (Supports%20Wraps :type Boolean)))

(fn SetBoostEnabled
  (params (BoostEnabled :type Boolean))
  (Variables|ChildFunction|SetBoostEnabled BoostEnabled))

(fn SetColorOption
  (params (ColorIndex :type Integer)))

(fn SetFuelValue
  (params (Value :type Float%20%28double-precision%29))
  (bind _asfort_athena_vehicle (Utilities|Casting|CastToFortAthenaVehicle (Variables|Default|GetCachedVehicleInterface)))
  (Class|FuelComponent|SetFuel (Vehicle|Fuel|GetVehicleFuelComponent _asfort_athena_vehicle) (Utilities|Operators|Multiply (Math|Float|Clamp(Float) Value 0.0 1e+02) 0.01 (Class|FuelComponent|GetFuelCapacity (Vehicle|Fuel|GetVehicleFuelComponent _asfort_athena_vehicle)))))

(fn SetMaxRandomStartingFuel
  (params (Max%20Fuel%20Percent :type Integer)))

(fn SetMinRandomStartingFuel
  (params (Min%20Fuel%20Percent :type Integer)))

(fn SetVehicleHealth
  (params (Value :type Float%20%28double-precision%29))
  (bind _asfort_athena_vehicle (Utilities|Casting|CastToFortAthenaVehicle (Variables|Default|GetCachedVehicleInterface)))
  (Class|FortVehicleInterface|SetHealthPercent _asfort_athena_vehicle (Math|Float|Clamp(Float) (/ (Math|Float|Clamp(Float) Value 1.0 9999.0) (Class|FortVehicleInterface|GetMaxHealth _asfort_athena_vehicle)) 1e-05)))

(fn SetVehicleMaxHealth
  (params (Value :type Float%20%28double-precision%29))
  (Networking|SwitchHasAuthority
    (:Authority
      (bind _asfort_athena_vehicle (Utilities|Casting|CastToFortAthenaVehicle (Variables|Default|GetCachedVehicleInterface)))
      (Utilities|IsValid _asfort_athena_vehicle
        (:Is%20Valid
          (if (not (Variables|UserOptions|GetVehicle%20Indestructible))
            (Utilities|IsValid (Ability|GetFortAbilitySystemComponent _asfort_athena_vehicle)
              (:Is%20Valid
                (GameplayEffects|RemoveActiveEffectswithAppliedTags (Ability|GetFortAbilitySystemComponent _asfort_athena_vehicle) "(GameplayTags=((TagName=\"Granted.Creative.CreativeEffect.Health\")))")
                (bind _returnvalue (Ability|GameplayEffect|AssignTagSetbyCallerMagnitude (GameplayEffects|MakeOutgoingSpec :self (Ability|GetFortAbilitySystemComponent _asfort_athena_vehicle) :GameplayEffectClass "/Game/Creative/Devices/Common/GameplayEffects/GE_Devices_Common_SetMaxHealth.GE_Devices_Common_SetMaxHealth_C" :Level 0.0) "(TagName=\"SetByCaller.MaxHealth\")" (Math|Float|Clamp(Float) (Variables|GetValue) 1.0 9999.0)))
                (GameplayEffects|ApplyGameplayEffectSpecToSelf (Ability|GetFortAbilitySystemComponent _asfort_athena_vehicle) _returnvalue)))))))))

(fn ShouldBlockPlacement
  (params (MoveTool :type Fort%20Creative%20Move%20Tool%20Object%20Reference))
  (bind _returnvalue (CreativeMoveTool|GetCurrentHitTraceRule MoveTool))
  (switch Utilities|FlowControl|Switch|SwitchonEHitTraceRule _returnvalue
    (:Visibility
      _)
    (:Terrain
      _)
    (:None
      ;[1 1 1] Block placement for any overlapped device or volume without "ignorevehicleplacementblock" tag
      (return false)))
  (bind _outactors (Collision|ComponentOverlapActors (Variables|Default|GetBox) (Transformation|GetWorldTransform (Variables|Default|GetBox)) (Utilities|Select :Index _returnvalue :Visibility (Utilities|Array|MakeArray "ObjectTypeQuery1" "ObjectTypeQuery2" "ObjectTypeQuery3" "ObjectTypeQuery4" "ObjectTypeQuery5" "ObjectTypeQuery6") :Terrain (Utilities|Array|MakeArray "ObjectTypeQuery1") :None "ObjectTypeQuery1") 0 (Utilities|Array|MakeArray (Components|GetOwner (Variables|Default|GetBox)) self)))
  (bind _array_element (Utilities|Array|ForEachLoopwithBreak _outactors)
    (:LoopBody
      (return (not (Actor|ActorHasTag _array_element "ignorevehicleplacementblock"))))
    (:Completed)))

(fn SpawnPointInVolume
  (Utilities|IsLocationRelevanttoMe (Transformation|GetWorldLocation (Variables|Default|GetSpawnLocation))
    (:IsRelevant
      (bind _area_clear (CallFunction|TestVehicleOverlap))
      (if _area_clear
        (return true)
        (else
          _)))
    (:NotRelevant
      _))
  (return false))

(fn TestVehicleOverlap
  (bind (_origin _boxextent) (Collision|GetComponentBounds (Variables|Default|GetBox)))
  (Utilities|Array|Add (LocalVariables|GetIgnoreActors) self)
  (Utilities|IsValid (Utilities|ToObject(Interface) (Variables|Default|GetCachedVehicleInterface))
    (:Is%20Valid
      (bind _asactor (Utilities|Casting|CastToActor (Utilities|ToObject(Interface) (Variables|Default|GetCachedVehicleInterface)))
        (:then
          (Utilities|Array|Add (LocalVariables|GetIgnoreActors) _asactor))
        (:CastFailed
          _)))
    (:Is%20Not%20Valid
      _))
  (bind (_outactors _returnvalue) (Collision|BoxOverlapActors _origin _boxextent (Utilities|Array|MakeArray "ObjectTypeQuery5" "ObjectTypeQuery12") 0 (LocalVariables|GetIgnoreActors) :!class /Script/Engine.KismetSystemLibrary))
  (return (not _returnvalue)))

(fn ThrottleNewVehicle
  ;[0.4337 1 0.2604] Use the Creative Throttle Manager to throttle the spawn of a new vehicle
  (if (not (Variables|SpamProtection|GetWaiting%20For%20Throttle))
    (bind _returnvalue (Class|CreativeBlueprintLibrary|GetCreativeGeneralThrottleManager))
    (Utilities|IsValid _returnvalue
      (:Is%20Valid
        (Variables|SpamProtection|SetWaiting%20For%20Throttle true)
        (Class|FortCreativeGeneralThrottleManager|RequestThrottleObject _returnvalue self (Math|Conversions|ToInteger(Byte) (Utilities|Enum|LiteralenumEGeneralThrottleBuckets "NewEnumerator0")) (EventDispatchers|CreateEvent :!event Completed%20Throttle))
        (Variables|Default|SetLastCooldownStart 0.0)))))

(fn TransmitOnCleanup
  (params (Instigator :type Controller%20Object%20Reference))
  (GameplayEvent|BroadcastEvent (Variables|UserOptions|Events|GetOn%20Vehicle%20Is%20Destroyed) 0 (Utilities|Struct|MakeGameplayEventVehicleDestroyed Instigator)))

(fn UpdateDataRegistryfromCurveTable
  (params (InCurveTable :type Object%20Reference))
  (CallFunction|InitializeDataRegistryResolverScope)
  (ScalableFloat|AddCurveTabletoDataRegistrywithContext InCurveTable (Variables|Vehicle|Tunings|GetData%20Registry%20Name)))

(fn UpdateEffects
  (Variables|Setlocal%20Cooldown (CallFunction|OnCooldown))
  (Networking|CanExecuteCosmeticEvents
    (:True
      (Utilities|FlowControl|Sequence
        (:then_0
          (if (CallFunction|IsinPlayMode)
            ;[0.1068 0.7917 0.05773] Setup cooldown effect material on the floor spawner mesh
            (bind _mid_spawnplate (Variables|Cosmetics|GetMID%20SpawnPlate)
              (:then
                (Rendering|Material|SetMaterial (Variables|Debug|GetStaticMeshComponent) 0 _mid_spawnplate))
              (:else
                (Material|ReattachOrCreateMID (Variables|Debug|GetStaticMeshComponent) (Variables|Cosmetics|GetMID%20SpawnPlate))
                (Rendering|Material|SetTextureParameterValue (Variables|Cosmetics|GetMID%20SpawnPlate) "EmissiveTexture" (Variables|Cosmetics|GetEmissive%20Texture))))
            ;[0.8361 0.08416 1] Update the floor light material
            (bind _output_get (LocalVariables|SetLocal%20Color (Math|Color|SelectColor (Math|Color|SelectColor (Variables|ColorSwitches|GetColorCooldown) (Variables|ColorSwitches|GetEnabledColor) (Variables|Getlocal%20Cooldown)) (Variables|ColorSwitches|GetDisabledColor) (Class|CreativeEnabledComponent|IsEnabled (Variables|Default|GetEnabledComponent)))))
            (CallFunction|BlendEmissiveColor _output_get) ; Interpolate the material color from the current state to the new state
            (Rendering|Material|SetScalarParameterValue (Variables|Cosmetics|GetMID%20SpawnPlate) "FlickerScale" (select (Variables|Getlocal%20Cooldown) (Variables|Cosmetics|GetSpawn%20Plate%20Flicker%20Scale) 0.0) :!class /Script/Engine.MaterialInstanceDynamic) ; Enable or disable a flash effect by changing the FlickerScale
          ))
        (:then_1
          ;[0.1085 0.9003 1] Play the ready sound, but not while in editor.
          (if (and (Class|CreativeEnabledComponent|IsEnabled (Variables|Default|GetEnabledComponent)) (not (Editor|InEditor)))
            (Audio|PlaySoundatLocation (Variables|Cosmetics|GetTrap%20Ready%20Sound) (Transformation|GetActorLocation))))))))

(fn UpdateVehicleConfigs
  (bind _consumefuel (ChildFunctions|GetConsumesFuel))
  (bind _activating_team (Variables|UserOptions|GetActivating%20Team))
  (bind _allowed_class (Variables|UserOptions|GetAllowed%20Class))
  ;[0.4608 1 0.3476 0.4] Set the vehicle runtime settings based on the devices user options
  (bind _asfort_athena_vehicle (Utilities|Casting|CastToFortAthenaVehicle (Variables|Default|GetCachedVehicleInterface)))
  (Class|FortAthenaVehicle|SetEnforceTeamRestriction _asfort_athena_vehicle (and (not (Creative|TeamOption|IsTeamType _activating_team "Any")) (CallFunction|IsinPlayMode)))
  (Class|FortAthenaVehicle|SetMinigameTeamID _asfort_athena_vehicle (select (Creative|TeamOption|IsTeamType _activating_team) (Teams|IndextoTeam (Knobs|Team|GetTeamValueasIndex (Creative|TeamOption|GetLegacyTeamIndex _activating_team))) (Utilities|Enum|LiteralenumEFortTeam "Spectator")))
  (Class|FortAthenaVehicle|SetEnforceMinigameClassRestriction _asfort_athena_vehicle (and (not (Creative|ClassOption|IsClassType _allowed_class "All")) (CallFunction|IsinPlayMode)))
  (Class|FortAthenaVehicle|SetMinigameClassID _asfort_athena_vehicle (Creative|ClassOption|GetLegacyClassIndex _allowed_class))
  (Class|FortAthenaVehicle|SetWaterDestructionDelay _asfort_athena_vehicle (select (Variables|UserOptions|GetDestroy%20when%20Stuck%20Underwater) (Variables|UserOptions|GetWater%20Destruction%20Timer) -1.0))
  (Class|FortAthenaVehicle|SetForceInfiniteFuel _asfort_athena_vehicle (not _consumefuel)) ; Fuel support still forthcoming, so force it to be infinite for now.
  (if (and (Utilities|IsValid (Vehicle|Fuel|GetVehicleFuelComponent _asfort_athena_vehicle) :!class /Script/Engine.KismetSystemLibrary) _consumefuel)
    (if (ChildFunctions|GetRandomStartingFuel)
      ;;[1 1 1 0.5] Set fuel from user set min and max
      (if (< (ChildFunctions|GetMinRandomStartingFuel) (ChildFunctions|GetMaxRandomStartingFuel))
        (Class|FuelComponent|SetFuel (Vehicle|Fuel|GetVehicleFuelComponent _asfort_athena_vehicle) (Utilities|Operators|Multiply (Math|Random|RandomFloatinRange (Math|Conversions|ToFloat(Integer) (ChildFunctions|GetMinRandomStartingFuel)) (Math|Conversions|ToFloat(Integer) (ChildFunctions|GetMaxRandomStartingFuel))) 0.01 (Class|FuelComponent|GetFuelCapacity (Vehicle|Fuel|GetVehicleFuelComponent _asfort_athena_vehicle))))
        (else
          _))
      (Class|FuelComponent|SetFuelUsageInfo (Vehicle|Fuel|GetVehicleFuelComponent _asfort_athena_vehicle) (Utilities|Struct|MakeRuntimeFuelUsageInfo (* (ChildFunctions|GetFuelUseMultiplier) (Utilities|Struct|BreakRuntimeFuelUsageInfo (Class|FuelComponent|GetFuelUsageInfo (Vehicle|Fuel|GetVehicleFuelComponent _asfort_athena_vehicle)))) (* (.FuelPerSecondDriving (Utilities|Struct|BreakRuntimeFuelUsageInfo (Class|FuelComponent|GetFuelUsageInfo (Vehicle|Fuel|GetVehicleFuelComponent _asfort_athena_vehicle)))) (ChildFunctions|GetFuelUseMultiplier)) (select (ChildFunctions|GetBoostEnabled) (* (.FuelPerSecondBoosting (Utilities|Struct|BreakRuntimeFuelUsageInfo (Class|FuelComponent|GetFuelUsageInfo (Vehicle|Fuel|GetVehicleFuelComponent _asfort_athena_vehicle)))) (ChildFunctions|GetBoostFuelUseMultiplier)) (* (.FuelPerSecondDriving (Utilities|Struct|BreakRuntimeFuelUsageInfo (Class|FuelComponent|GetFuelUsageInfo (Vehicle|Fuel|GetVehicleFuelComponent _asfort_athena_vehicle)))) (ChildFunctions|GetFuelUseMultiplier)))))
      (else
        (Class|FuelComponent|SetFuel (Vehicle|Fuel|GetVehicleFuelComponent _asfort_athena_vehicle) (Utilities|Operators|Multiply (Math|Conversions|ToFloat(Integer) (ChildFunctions|GetStartingFuelValue)) 0.01 (Class|FuelComponent|GetFuelCapacity (Vehicle|Fuel|GetVehicleFuelComponent _asfort_athena_vehicle))))))))

(fn UpdateVehicleSpawnSettings
  (params (EnableOrDisableReceiver :type Boolean))
  ;[0 1 0.3818] Respawn or destroy the vehicle when bEnabled changes or options are changed
  (Creative|Networking|Authority/Cosmetic
    (:Authority
      (Utilities|Time|ClearandInvalidateTimerbyHandle (Variables|Default|GetVehicle%20Spawn%20Timer))
      (if (Class|CreativeEnabledComponent|IsEnabled (Variables|Default|GetEnabledComponent))
        (if (Math|Boolean|ANDBoolean (or (Utilities|Enum|Equal(Enum) (Variables|UserOptions|GetRespawn%20Vehicle%20When%20Enabled) "NewEnumerator2") (and (Utilities|Enum|Equal(Enum) (Variables|UserOptions|GetRespawn%20Vehicle%20When%20Enabled) "NewEnumerator1") (not (Utilities|IsValid (Utilities|ToObject(Interface) (Variables|Default|GetCachedVehicleInterface)) :!class /Script/Engine.KismetSystemLibrary)))) (not (CallFunction|OnCooldown)) (CallFunction|SpawnPointinVolume))
          (CallFunction|RespawnVehicleFromSettings)
          (elif (Variables|GetEnableOrDisableReceiver)
            (Variables|Default|SetCachedInstigator 0)))
        (elif (Variables|UserOptions|GetDestroy%20Vehicle%20When%20Disabled)
          (CallFunction|ManuallyDestroyVehicle)
          (if (Variables|GetEnableOrDisableReceiver)
            (Variables|Default|SetCachedVehicleInterface 0)
            (GameplayEvent|BroadcastEvent (Variables|UserOptions|Events|GetOn%20Vehicle%20Is%20Destroyed) 0 (Utilities|Struct|MakeGameplayEventVehicleDestroyed (Variables|Default|GetCachedInstigator)))
            (Variables|Default|SetCachedInstigator 0)))))
    (:Cosmetic
      (CallFunction|UpdateEffects))))

(fn ConstructionScript
  (|CallParentFunction))

(fn VolumeClearanceCheck)
